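# Physics asset for the Prisoner male medium body's dynamic skirt, apparently the skydiving variant, with a small hip ring.
# Capsule bone chains hang from the pelvis at the back and the thigh twist bones at front and sides, alongside leg, pelvis and spine capsules and a flat root box.
# Chain joints mostly carry swing/twist limits, but the upper right-side chain is free.
# Cross-links tie neighbouring chains on the left side only.
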PhysicsAsset /BRCosmetics/Characters/Player/Male/Medium/Bodies/M_MED_Prisoner/Meshes/M_MED_Prisoner_Dyn_Skirt_Skydiving
bodies {
 calf_l [{shapeType=Capsule, length=35, radius=10, center=<-28 -1 0>, rotation=<rot 90 0 0>}]
 calf_r [{shapeType=Capsule, length=34.4819, radius=11.5365, center=<28.4281 1.6068 0.8734>, rotation=<rot 86.7735 180 180>}]
 dyn_cloth_bl_01 [{shapeType=Capsule, length=10, radius=4, shapeName=dyn_cloth_bl_01_capsule, center=<4 0 0>}]
 dyn_cloth_bl_02 [{shapeType=Capsule, length=9, radius=8, shapeName=dyn_cloth_bl_02_capsule, center=<10 4 0>}]
 dyn_cloth_bl_03 [{shapeType=Capsule, length=5, radius=10, shapeName=dyn_cloth_bl_03_capsule, center=<6 4 0>}]
 dyn_cloth_bl_04 [{shapeType=Capsule, length=15, radius=10, shapeName=dyn_cloth_bl_04_capsule, center=<7 4 0>}]
 dyn_cloth_bl_05 [{shapeType=Capsule, length=15, radius=10, shapeName=dyn_cloth_bl_05_capsule, center=<5 4 0>}]
 dyn_cloth_bl_mid_01 [{shapeType=Capsule, length=10, radius=4, shapeName=dyn_cloth_bl_mid_01_capsule, center=<4 0 0>}]
 dyn_cloth_bl_mid_02 [{shapeType=Capsule, length=6, radius=8, shapeName=dyn_cloth_bl_mid_02_capsule, center=<11 6.5 0>}]
 dyn_cloth_bl_mid_03 [{shapeType=Capsule, length=11, radius=8, shapeName=dyn_cloth_br_mid_03_capsule, center=<7 7 0>}]
 dyn_cloth_bl_mid_04 [{shapeType=Capsule, length=15, radius=10, shapeName=dyn_cloth_br_mid_04_capsule, center=<6 7 0>}]
 dyn_cloth_bl_mid_05 [{shapeType=Capsule, length=15, radius=10, shapeName=dyn_cloth_br_mid_05_capsule, center=<3 7 0>}]
 dyn_cloth_bl_mid_07 [{shapeType=Capsule, length=15, radius=10, shapeName=dyn_cloth_br_mid_07_capsule, center=<5 7 0>}]
 dyn_cloth_br_01 [{shapeType=Capsule, length=11, radius=4, shapeName=dyn_cloth_br_01_capsule, center=<3.5 0 0>}]
 dyn_cloth_br_02 [{shapeType=Capsule, length=11, radius=7, shapeName=dyn_cloth_br_02_capsule, center=<8 5 0>}]
 dyn_cloth_br_03 [{shapeType=Capsule, length=7, radius=10, shapeName=dyn_cloth_br_03_capsule, center=<5.5 5 0>}]
 dyn_cloth_br_04 [{shapeType=Capsule, length=15, radius=10, shapeName=dyn_cloth_br_04_capsule, center=<6.5 5 0>}]
 dyn_cloth_br_05 [{shapeType=Capsule, length=15, radius=10, shapeName=dyn_cloth_br_05_capsule, center=<6 5 0>}]
 dyn_cloth_br_mid_01 [{shapeType=Capsule, length=10, radius=4, shapeName=dyn_cloth_br_mid_01_capsule, center=<3 0 0>}]
 dyn_cloth_br_mid_02 [{shapeType=Capsule, length=8, radius=8, shapeName=dyn_cloth_br_mid_02_capsule, center=<7.5 7 0>}]
 dyn_cloth_br_mid_03 [{shapeType=Capsule, length=11, radius=8, shapeName=dyn_cloth_br_mid_03_capsule, center=<7 8 0>}]
 dyn_cloth_br_mid_04 [{shapeType=Capsule, length=15, radius=10, shapeName=dyn_cloth_br_mid_04_capsule, center=<6 8 0>}]
 dyn_cloth_br_mid_05 [{shapeType=Capsule, length=15, radius=10, shapeName=dyn_cloth_br_mid_05_capsule, center=<3 8 0>}]
 dyn_cloth_br_mid_07 [{shapeType=Capsule, length=15, radius=10, shapeName=dyn_cloth_br_mid_07_capsule, center=<5 8 0>}]
 dyn_cloth_fl_01 [{shapeType=Capsule, length=8, radius=4, shapeName=dyn_cloth_fl_01_capsule, center=<7.5 0 0>}]
 dyn_cloth_fl_02 [{shapeType=Capsule, length=8, radius=8, shapeName=dyn_cloth_fl_02_capsule, center=<8 6 0>}]
 dyn_cloth_fl_03 [{shapeType=Capsule, length=6, radius=8, shapeName=dyn_cloth_fl_03_capsule, center=<6 5 0>}]
 dyn_cloth_fl_04 [{shapeType=Capsule, length=15, radius=10, shapeName=dyn_cloth_fl_04_capsule, center=<6 7 0>}]
 dyn_cloth_fl_05 [{shapeType=Capsule, length=15, radius=10, shapeName=dyn_cloth_fl_05_capsule, center=<5 7 0>}]
 dyn_cloth_fl_06 [{shapeType=Capsule, length=15, radius=10, shapeName=dyn_cloth_fl_06_capsule, center=<5 7 0>}]
 dyn_cloth_fl_MSTR [{shapeType=Capsule, length=0.505, radius=0.505, rotation=<rot 0 -90 0>}]
 dyn_cloth_fr_01 [{shapeType=Capsule, length=7, radius=4, shapeName=dyn_cloth_fr_01_capsule, center=<5.5 0 0>}]
 dyn_cloth_fr_02 [{shapeType=Capsule, length=6, radius=8, shapeName=dyn_cloth_fr_02_capsule, center=<9 7 0>}]
 dyn_cloth_fr_03 [{shapeType=Capsule, length=6, radius=8, shapeName=dyn_cloth_fr_03_capsule, center=<6 7 0>}]
 dyn_cloth_fr_04 [{shapeType=Capsule, length=15, radius=10, shapeName=dyn_cloth_fr_04_capsule, center=<6 7 0>}]
 dyn_cloth_fr_05 [{shapeType=Capsule, length=15, radius=10, shapeName=dyn_cloth_fr_05_capsule, center=<7 7 0>}]
 dyn_cloth_fr_06 [{shapeType=Capsule, length=15, radius=10, shapeName=dyn_cloth_fr_06_capsule, center=<5.5 7 0>}]
 dyn_cloth_fr_MSTR [{shapeType=Capsule, length=0.505, radius=0.505, rotation=<rot 0 -90 0>}]
 dyn_cloth_l_01 [{shapeType=Capsule, length=16.505, radius=4.255, shapeName=dyn_cloth_l_01_capsule, center=<3.2218 0.0001 0>, rotation=<rot 0 -90 0>}]
 dyn_cloth_l_02 [{
   shapeType=Capsule
   length=17.4768
   radius=5.0265
   shapeName=dyn_cloth_l_02_capsule
   center=<7.8153 0.4313 -0.1951>
   rotation=<rot 0.0001 -83.5187 -176.9043>
  }]
 dyn_cloth_l_03 [{shapeType=Capsule, length=13, radius=8, shapeName=dyn_cloth_l_03_capsule, center=<4 4 0>}]
 dyn_cloth_l_04 [{shapeType=Capsule, length=15, radius=10, shapeName=dyn_cloth_l_04_capsule, center=<7 8 0>}]
 dyn_cloth_l_05 [{shapeType=Capsule, length=15, radius=10, shapeName=dyn_cloth_l_05_capsule, center=<8 8 0>}]
 dyn_cloth_l_06 [{shapeType=Capsule, length=15, radius=10, shapeName=dyn_cloth_l_06_capsule, center=<6 8 0>}]
 dyn_cloth_l_mid_02 [{shapeType=Capsule, length=7, radius=8, shapeName=dyn_cloth_l_mid_02_capsule, center=<8 9 0>}]
 dyn_cloth_l_mid_03 [{shapeType=Capsule, length=7, radius=8, shapeName=dyn_cloth_l_mid_03_capsule, center=<4 9 0>}]
 dyn_cloth_l_mid_04 [{shapeType=Capsule, length=15, radius=10, shapeName=dyn_cloth_l_mid_04_capsule, center=<6 7 0>}]
 dyn_cloth_l_mid_05 [{shapeType=Capsule, length=15, radius=10, shapeName=dyn_cloth_l_mid_05_capsule, center=<5.5 7 0>}]
 dyn_cloth_l_mid_07 [{shapeType=Capsule, length=15, radius=10, shapeName=dyn_cloth_l_mid_07_capsule, center=<7 7 0>}]
 dyn_cloth_l_mid_08 [{shapeType=Capsule, length=6, radius=4, shapeName=dyn_cloth_l_mid_08_capsule, center=<4 0 0>}]
 dyn_cloth_r_01 [{
   shapeType=Capsule
   length=11.5635
   radius=4.8319
   shapeName=dyn_cloth_r_01_capsule
   center=<2.8996 0.4059 0.5414>
   rotation=<rot 1.6349 -85.4853 4.4976>
  }]
 dyn_cloth_r_02 [{
   shapeType=Capsule
   length=8.8052
   radius=5.8535
   shapeName=dyn_cloth_r_02_capsule
   center=<8.723 0.2557 -1.4328>
   rotation=<rot -1.0304 96.885 -5.855>
  }]
 dyn_cloth_r_03 [{
   shapeType=Capsule
   length=8.7406
   radius=5.425
   shapeName=dyn_cloth_r_03_capsule
   center=<5.4773 2.7704 -3.4443>
   rotation=<rot 0 -95.0003 3.3563>
  }]
 dyn_cloth_r_04 [{shapeType=Capsule, length=15, radius=10, shapeName=dyn_cloth_r_04_capsule, center=<6 9 0>}]
 dyn_cloth_r_05 [{shapeType=Capsule, length=15, radius=10, shapeName=dyn_cloth_r_05_capsule, center=<5 9 0>}]
 dyn_cloth_r_06 [{shapeType=Capsule, length=15, radius=10, shapeName=dyn_cloth_r_06_capsule, center=<5.5 9 0>}]
 dyn_cloth_r_mid_02 [{shapeType=Capsule, length=7, radius=8, shapeName=dyn_cloth_r_mid_02_capsule, center=<9 7 0>}]
 dyn_cloth_r_mid_03 [{shapeType=Capsule, length=6, radius=8, shapeName=dyn_cloth_r_mid_03_capsule, center=<6 7 0>}]
 dyn_cloth_r_mid_04 [{shapeType=Capsule, length=7.5, radius=10, shapeName=dyn_cloth_r_mid_04_capsule, center=<6 7 0>}]
 dyn_cloth_r_mid_05 [{shapeType=Capsule, length=15, radius=10, shapeName=dyn_cloth_r_05_capsule, center=<5 7 0>}]
 dyn_cloth_r_mid_07 [{shapeType=Capsule, length=15, radius=10, shapeName=dyn_cloth_r_06_capsule, center=<5.5 7 0>}]
 dyn_cloth_r_mid_09 [{shapeType=Capsule, length=7, radius=4, shapeName=dyn_cloth_r_mid_09_capsule, center=<4.5 0 0>}]
 dyn_hip_ring [{shapeType=Capsule, length=3.4181, radius=2.2544, center=<4.8168 0 -0.0002>, rotation=<rot 90 0 0>}]
 dyn_hip_ring_MSTR [{shapeType=Capsule, length=0.505, radius=0.505, rotation=<rot 0 -90 0>}]
 dyn_side_ring_MSTR [{shapeType=Capsule, length=0.505, radius=0.505, rotation=<rot 0 -90 0>}]
 pelvis [{shapeType=Capsule, length=9.8119, radius=9.8283}]
 root [{shapeType=Box, extentX=240.0799, extentY=255.951, extentZ=19.18, center=<0 -0.0001 -7.3631>}]
 spine_01 [{shapeType=Capsule, length=2, radius=2}]
 spine_02 [{shapeType=Capsule, length=2, radius=2}]
 thigh_l [{shapeType=Capsule, length=19, radius=10, center=<-26 -1 0>, rotation=<rot 90 0 0>}]
 thigh_r [{shapeType=Capsule, length=22.2951, radius=12.4426, center=<27.9634 0.9073 -4.1192>, rotation=<rot 86.1435 -177.9 -179.3222>}]
 thigh_twist_01_l [{shapeType=Capsule, length=2, radius=2}]
 thigh_twist_01_r [{shapeType=Capsule, length=2, radius=2}]
}
constraints [
 {bone1Name=dyn_cloth_l_02, bone2Name=dyn_cloth_l_01, swing1LimitDegrees=75, swing2LimitDegrees=75, twistLimitDegrees=15}
 {bone1Name=dyn_cloth_l_03, bone2Name=dyn_cloth_l_02, swing1LimitDegrees=75, swing2LimitDegrees=75, twistLimitDegrees=15}
 {bone1Name=dyn_cloth_l_04, bone2Name=dyn_cloth_l_03, swing1LimitDegrees=30, swing2LimitDegrees=30, twistLimitDegrees=15}
 {bone1Name=dyn_cloth_l_05, bone2Name=dyn_cloth_l_04, swing1LimitDegrees=30, swing2LimitDegrees=30, twistLimitDegrees=15}
 {bone1Name=dyn_cloth_l_06, bone2Name=dyn_cloth_l_05, swing1LimitDegrees=30, swing2LimitDegrees=30, twistLimitDegrees=15}
 {bone1Name=dyn_cloth_fl_01, bone2Name=dyn_cloth_fl_MSTR, swing1LimitDegrees=30, swing2LimitDegrees=30, twistLimitDegrees=15}
 {bone1Name=dyn_cloth_fl_02, bone2Name=dyn_cloth_fl_01, swing1LimitDegrees=30, swing2LimitDegrees=30, twistLimitDegrees=15}
 {bone1Name=dyn_cloth_fl_03, bone2Name=dyn_cloth_fl_02, swing1LimitDegrees=30, swing2LimitDegrees=30, twistLimitDegrees=15}
 {bone1Name=dyn_cloth_fl_04, bone2Name=dyn_cloth_fl_03, swing1LimitDegrees=30, swing2LimitDegrees=30, twistLimitDegrees=15}
 {bone1Name=dyn_cloth_fl_05, bone2Name=dyn_cloth_fl_04, swing1LimitDegrees=30, swing2LimitDegrees=30, twistLimitDegrees=15}
 {bone1Name=dyn_cloth_fl_06, bone2Name=dyn_cloth_fl_05, swing1LimitDegrees=30, swing2LimitDegrees=30, twistLimitDegrees=15}
 {bone1Name=dyn_cloth_bl_01, bone2Name=pelvis, swing1LimitDegrees=30, swing2LimitDegrees=30, twistLimitDegrees=15}
 {bone1Name=dyn_cloth_bl_02, bone2Name=dyn_cloth_bl_01, swing1LimitDegrees=30, swing2LimitDegrees=30, twistLimitDegrees=15}
 {bone1Name=dyn_cloth_bl_03, bone2Name=dyn_cloth_bl_02, swing1LimitDegrees=30, swing2LimitDegrees=30, twistLimitDegrees=15}
 {bone1Name=dyn_cloth_bl_04, bone2Name=dyn_cloth_bl_03, swing1LimitDegrees=30, swing2LimitDegrees=30, twistLimitDegrees=15}
 {bone1Name=dyn_cloth_bl_05, bone2Name=dyn_cloth_bl_04, swing1LimitDegrees=30, swing2LimitDegrees=30, twistLimitDegrees=15}
 {bone1Name=dyn_cloth_bl_mid_01, bone2Name=pelvis, swing1LimitDegrees=30, swing2LimitDegrees=30, twistLimitDegrees=15}
 {bone1Name=dyn_cloth_bl_mid_02, bone2Name=dyn_cloth_bl_mid_01, swing1LimitDegrees=30, swing2LimitDegrees=30, twistLimitDegrees=15}
 {bone1Name=dyn_cloth_bl_mid_03, bone2Name=dyn_cloth_bl_mid_02, swing1LimitDegrees=30, swing2LimitDegrees=30, twistLimitDegrees=15}
 {bone1Name=dyn_cloth_bl_mid_04, bone2Name=dyn_cloth_bl_mid_03, swing1LimitDegrees=30, swing2LimitDegrees=30, twistLimitDegrees=15}
 {bone1Name=dyn_cloth_bl_mid_05, bone2Name=dyn_cloth_bl_mid_04, swing1LimitDegrees=30, swing2LimitDegrees=30, twistLimitDegrees=15}
 {bone1Name=dyn_cloth_bl_mid_07, bone2Name=dyn_cloth_bl_mid_05, swing1LimitDegrees=30, swing2LimitDegrees=30, twistLimitDegrees=15}
 {bone1Name=dyn_cloth_l_mid_08, bone2Name=thigh_twist_01_l, swing1LimitDegrees=30, swing2LimitDegrees=30, twistLimitDegrees=15}
 {bone1Name=dyn_cloth_l_mid_02, bone2Name=dyn_cloth_l_mid_08, swing1LimitDegrees=30, swing2LimitDegrees=30, twistLimitDegrees=15}
 {bone1Name=dyn_cloth_l_mid_03, bone2Name=dyn_cloth_l_mid_02, swing1LimitDegrees=30, swing2LimitDegrees=30, twistLimitDegrees=15}
 {bone1Name=dyn_cloth_l_mid_04, bone2Name=dyn_cloth_l_mid_03, swing1LimitDegrees=30, swing2LimitDegrees=30, twistLimitDegrees=15}
 {bone1Name=dyn_cloth_l_mid_05, bone2Name=dyn_cloth_l_mid_04, swing1LimitDegrees=30, swing2LimitDegrees=30, twistLimitDegrees=15}
 {bone1Name=dyn_cloth_l_mid_07, bone2Name=dyn_cloth_l_mid_05, swing1LimitDegrees=30, swing2LimitDegrees=30, twistLimitDegrees=15}
 {
  bone1Name=thigh_l
  bone2Name=pelvis
  swing1LimitDegrees=45
  swing1Motion=Free
  swing2LimitDegrees=45
  swing2Motion=Free
  twistLimitDegrees=45
  twistMotion=Free
 }
 {
  bone1Name=calf_l
  bone2Name=thigh_l
  swing1LimitDegrees=45
  swing1Motion=Free
  swing2LimitDegrees=45
  swing2Motion=Free
  twistLimitDegrees=45
  twistMotion=Free
 }
 {
  bone1Name=thigh_r
  bone2Name=pelvis
  swing1LimitDegrees=45
  swing1Motion=Free
  swing2LimitDegrees=45
  swing2Motion=Free
  twistLimitDegrees=45
  twistMotion=Free
 }
 {
  bone1Name=calf_r
  bone2Name=thigh_r
  swing1LimitDegrees=45
  swing1Motion=Free
  swing2LimitDegrees=45
  swing2Motion=Free
  twistLimitDegrees=45
  twistMotion=Free
 }
 {bone1Name=dyn_cloth_l_01, bone2Name=thigh_twist_01_l, swing1LimitDegrees=75, swing2LimitDegrees=75, twistLimitDegrees=15}
 {
  bone1Name=spine_01
  bone2Name=pelvis
  swing1LimitDegrees=45
  swing1Motion=Free
  swing2LimitDegrees=45
  swing2Motion=Free
  twistLimitDegrees=45
  twistMotion=Free
 }
 {
  bone1Name=thigh_twist_01_l
  bone2Name=thigh_l
  swing1LimitDegrees=45
  swing1Motion=Free
  swing2LimitDegrees=45
  swing2Motion=Free
  twistLimitDegrees=45
  twistMotion=Free
 }
 {
  bone1Name=thigh_twist_01_r
  bone2Name=thigh_r
  swing1LimitDegrees=45
  swing1Motion=Free
  swing2LimitDegrees=45
  swing2Motion=Free
  twistLimitDegrees=45
  twistMotion=Free
 }
 {bone1Name=dyn_cloth_l_mid_05, bone2Name=dyn_cloth_fl_05, swing1LimitDegrees=45, swing2LimitDegrees=45, twistLimitDegrees=45}
 {bone1Name=dyn_cloth_l_mid_03, bone2Name=dyn_cloth_fl_03, swing1LimitDegrees=45, swing2LimitDegrees=45, twistLimitDegrees=45}
 {
  bone1Name=pelvis
  bone2Name=root
  swing1LimitDegrees=45
  swing1Motion=Free
  swing2LimitDegrees=45
  swing2Motion=Free
  twistLimitDegrees=45
  twistMotion=Free
 }
 {bone1Name=dyn_cloth_bl_mid_03, bone2Name=dyn_cloth_bl_03, swing1LimitDegrees=45, swing2LimitDegrees=45, twistLimitDegrees=45}
 {bone1Name=dyn_cloth_bl_mid_05, bone2Name=dyn_cloth_bl_05, swing1LimitDegrees=45, swing2LimitDegrees=45, twistLimitDegrees=45}
 {bone1Name=dyn_cloth_l_mid_03, bone2Name=dyn_cloth_l_03, swing1LimitDegrees=45, swing2LimitDegrees=45, twistLimitDegrees=45}
 {bone1Name=dyn_cloth_l_mid_05, bone2Name=dyn_cloth_l_05, swing1LimitDegrees=45, swing2LimitDegrees=45, twistLimitDegrees=45}
 {
  bone1Name=dyn_cloth_fl_MSTR
  bone2Name=thigh_twist_01_l
  swing1LimitDegrees=45
  swing1Motion=Free
  swing2LimitDegrees=45
  swing2Motion=Free
  twistLimitDegrees=45
  twistMotion=Free
 }
 {bone1Name=dyn_cloth_bl_mid_04, bone2Name=dyn_cloth_l_04, swing1LimitDegrees=45, swing2LimitDegrees=45, twistLimitDegrees=45}
 {
  bone1Name=spine_02
  bone2Name=spine_01
  swing1LimitDegrees=45
  swing1Motion=Free
  swing2LimitDegrees=45
  swing2Motion=Free
  twistLimitDegrees=45
  twistMotion=Free
 }
 {
  bone1Name=dyn_cloth_fl_05
  bone2Name=dyn_cloth_l_05
  swing1LimitDegrees=45
  swing1Motion=Free
  swing2LimitDegrees=45
  swing2Motion=Free
  twistLimitDegrees=45
  twistMotion=Free
 }
 {
  bone1Name=dyn_cloth_fl_06
  bone2Name=dyn_cloth_l_06
  swing1LimitDegrees=45
  swing1Motion=Free
  swing2LimitDegrees=45
  swing2Motion=Free
  twistLimitDegrees=45
  twistMotion=Free
 }
 {
  bone1Name=dyn_cloth_fl_04
  bone2Name=dyn_cloth_l_04
  swing1LimitDegrees=45
  swing1Motion=Free
  swing2LimitDegrees=45
  swing2Motion=Free
  twistLimitDegrees=45
  twistMotion=Free
 }
 {
  bone1Name=dyn_cloth_fl_03
  bone2Name=dyn_cloth_l_03
  swing1LimitDegrees=45
  swing1Motion=Free
  swing2LimitDegrees=45
  swing2Motion=Free
  twistLimitDegrees=45
  twistMotion=Free
 }
 {bone1Name=dyn_cloth_r_mid_09, bone2Name=thigh_twist_01_r, swing1LimitDegrees=30, swing2LimitDegrees=30, twistLimitDegrees=15}
 {bone1Name=dyn_cloth_r_mid_02, bone2Name=dyn_cloth_r_mid_09, swing1LimitDegrees=30, swing2LimitDegrees=30, twistLimitDegrees=15}
 {bone1Name=dyn_cloth_r_mid_03, bone2Name=dyn_cloth_r_mid_02, swing1LimitDegrees=30, swing2LimitDegrees=30, twistLimitDegrees=15}
 {bone1Name=dyn_cloth_r_mid_04, bone2Name=dyn_cloth_r_mid_03, swing1LimitDegrees=30, swing2LimitDegrees=30, twistLimitDegrees=15}
 {bone1Name=dyn_cloth_r_mid_05, bone2Name=dyn_cloth_r_mid_04, swing1LimitDegrees=30, swing2LimitDegrees=30, twistLimitDegrees=15}
 {bone1Name=dyn_cloth_r_mid_07, bone2Name=dyn_cloth_r_mid_05, swing1LimitDegrees=30, swing2LimitDegrees=30, twistLimitDegrees=15}
 {
  bone1Name=dyn_cloth_r_01
  bone2Name=thigh_twist_01_r
  swing1LimitDegrees=45
  swing1Motion=Free
  swing2LimitDegrees=45
  swing2Motion=Free
  twistLimitDegrees=45
  twistMotion=Free
 }
 {
  bone1Name=dyn_cloth_r_02
  bone2Name=dyn_cloth_r_01
  swing1LimitDegrees=45
  swing1Motion=Free
  swing2LimitDegrees=45
  swing2Motion=Free
  twistLimitDegrees=45
  twistMotion=Free
 }
 {
  bone1Name=dyn_cloth_r_03
  bone2Name=dyn_cloth_r_02
  swing1LimitDegrees=45
  swing1Motion=Free
  swing2LimitDegrees=45
  swing2Motion=Free
  twistLimitDegrees=45
  twistMotion=Free
 }
 {bone1Name=dyn_cloth_r_04, bone2Name=dyn_cloth_r_03, swing1LimitDegrees=30, swing2LimitDegrees=30, twistLimitDegrees=15}
 {bone1Name=dyn_cloth_r_05, bone2Name=dyn_cloth_r_04, swing1LimitDegrees=30, swing2LimitDegrees=30, twistLimitDegrees=15}
 {bone1Name=dyn_cloth_r_06, bone2Name=dyn_cloth_r_05, swing1LimitDegrees=30, swing2LimitDegrees=30, twistLimitDegrees=15}
 {bone1Name=dyn_cloth_fr_01, bone2Name=dyn_cloth_fr_MSTR, swing1LimitDegrees=30, swing2LimitDegrees=30, twistLimitDegrees=15}
 {bone1Name=dyn_cloth_fr_02, bone2Name=dyn_cloth_fr_01, swing1LimitDegrees=30, swing2LimitDegrees=30, twistLimitDegrees=15}
 {bone1Name=dyn_cloth_fr_03, bone2Name=dyn_cloth_fr_02, swing1LimitDegrees=30, swing2LimitDegrees=30, twistLimitDegrees=15}
 {bone1Name=dyn_cloth_fr_04, bone2Name=dyn_cloth_fr_03, swing1LimitDegrees=30, swing2LimitDegrees=30, twistLimitDegrees=15}
 {bone1Name=dyn_cloth_fr_05, bone2Name=dyn_cloth_fr_04, swing1LimitDegrees=30, swing2LimitDegrees=30, twistLimitDegrees=15}
 {bone1Name=dyn_cloth_fr_06, bone2Name=dyn_cloth_fr_05, swing1LimitDegrees=30, swing2LimitDegrees=30, twistLimitDegrees=15}
 {
  bone1Name=dyn_cloth_fr_MSTR
  bone2Name=thigh_twist_01_r
  swing1LimitDegrees=45
  swing1Motion=Free
  swing2LimitDegrees=45
  swing2Motion=Free
  twistLimitDegrees=45
  twistMotion=Free
 }
 {bone1Name=dyn_cloth_br_01, bone2Name=pelvis, swing1LimitDegrees=30, swing2LimitDegrees=30, twistLimitDegrees=15}
 {bone1Name=dyn_cloth_br_02, bone2Name=dyn_cloth_br_01, swing1LimitDegrees=30, swing2LimitDegrees=30, twistLimitDegrees=15}
 {bone1Name=dyn_cloth_br_03, bone2Name=dyn_cloth_br_02, swing1LimitDegrees=30, swing2LimitDegrees=30, twistLimitDegrees=15}
 {bone1Name=dyn_cloth_br_04, bone2Name=dyn_cloth_br_03, swing1LimitDegrees=30, swing2LimitDegrees=30, twistLimitDegrees=15}
 {bone1Name=dyn_cloth_br_05, bone2Name=dyn_cloth_br_04, swing1LimitDegrees=30, swing2LimitDegrees=30, twistLimitDegrees=15}
 {bone1Name=dyn_cloth_br_mid_01, bone2Name=pelvis, swing1LimitDegrees=30, swing2LimitDegrees=30, twistLimitDegrees=15}
 {bone1Name=dyn_cloth_br_mid_02, bone2Name=dyn_cloth_br_mid_01, swing1LimitDegrees=30, swing2LimitDegrees=30, twistLimitDegrees=15}
 {bone1Name=dyn_cloth_br_mid_03, bone2Name=dyn_cloth_br_mid_02, swing1LimitDegrees=30, swing2LimitDegrees=30, twistLimitDegrees=15}
 {bone1Name=dyn_cloth_br_mid_04, bone2Name=dyn_cloth_br_mid_03, swing1LimitDegrees=30, swing2LimitDegrees=30, twistLimitDegrees=15}
 {bone1Name=dyn_cloth_br_mid_05, bone2Name=dyn_cloth_br_mid_04, swing1LimitDegrees=30, swing2LimitDegrees=30, twistLimitDegrees=15}
 {bone1Name=dyn_cloth_br_mid_07, bone2Name=dyn_cloth_br_mid_05, swing1LimitDegrees=30, swing2LimitDegrees=30, twistLimitDegrees=15}
 {
  bone1Name=dyn_side_ring_MSTR
  bone2Name=pelvis
  swing1LimitDegrees=45
  swing1Motion=Free
  swing2LimitDegrees=45
  swing2Motion=Free
  twistLimitDegrees=45
  twistMotion=Free
 }
 {
  bone1Name=dyn_hip_ring_MSTR
  bone2Name=pelvis
  swing1LimitDegrees=45
  swing1Motion=Free
  swing2LimitDegrees=45
  swing2Motion=Free
  twistLimitDegrees=45
  twistMotion=Free
 }
 {bone1Name=dyn_hip_ring, bone2Name=dyn_hip_ring_MSTR, swing1LimitDegrees=45, swing2LimitDegrees=45, twistLimitDegrees=15}
]
