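# Physics asset for the dynamic left skirt of the medium male Prisoner body, built from capsule bodies on cloth bones.
# Five bone chains with swing and twist limits hang from the pelvis (back) or left thigh twist bone (side and front).
# Cross constraints tie neighbouring chains into a sheet, some with free rotation.
# Capsules on pelvis, spine, thighs and calves, plus a wide flat root box at floor level, provide body and ground shapes.
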PhysicsAsset /BRCosmetics/Characters/Player/Male/Medium/Bodies/M_MED_Prisoner/Meshes/M_MED_Prisoner_Dyn_Skirt_l
bodies {
 calf_l [{shapeType=Capsule, length=34.4819, radius=11.5365, center=<-28.4281 -1.6068 -0.8734>, rotation=<rot 86.7735 180 180>}]
 calf_r [{shapeType=Capsule, length=34.4819, radius=11.5365, center=<28.4281 1.6068 0.8734>, rotation=<rot 86.7735 180 180>}]
 dyn_cloth_bl_01 [{shapeType=Capsule, length=10.005, radius=6.005, shapeName=dyn_cloth_bl_01_capsule, center=<3.9225 0.0003 0>, rotation=<rot 0 -90 0>}]
 dyn_cloth_bl_02 [{
   shapeType=Capsule
   length=8.9485
   radius=5.619
   shapeName=dyn_cloth_bl_02_capsule
   center=<10.3793 -0.296 -0.0748>
   rotation=<rot 8.0088 -89.9998 -3.1878>
  }]
 dyn_cloth_bl_03 [{
   shapeType=Capsule
   length=5.282
   radius=5.2671
   shapeName=dyn_cloth_bl_03_capsule
   center=<5.8129 -0.5677 -1.9475>
   rotation=<rot 0 92.4951 2.1997>
  }]
 dyn_cloth_bl_04 [{
   shapeType=Capsule
   length=5.4589
   radius=5.3591
   shapeName=dyn_cloth_bl_04_capsule
   center=<7.3885 -1.0693 -1.4044>
   rotation=<rot 14.4512 -89.9997 178.3244>
  }]
 dyn_cloth_bl_05 [{
   shapeType=Capsule
   length=5.558
   radius=5.25
   shapeName=dyn_cloth_bl_05_capsule
   center=<5.0851 -0.8251 0.2104>
   rotation=<rot 12.4626 -89.9999 -0.4978>
  }]
 dyn_cloth_bl_mid_01 [{shapeType=Capsule, length=10.005, radius=5.755, shapeName=dyn_cloth_bl_mid_01_capsule, center=<3.8152 0 0>, rotation=<rot 0 -90 0>}]
 dyn_cloth_bl_mid_02 [{
   shapeType=Capsule
   length=6.3384
   radius=6.3265
   shapeName=dyn_cloth_bl_mid_02_capsule
   center=<10.6741 1.4754 0.1191>
   rotation=<rot 2.4561 -90.2152 0.5815>
  }]
 dyn_cloth_bl_mid_03 [{
   shapeType=Capsule
   length=7.0187
   radius=5.8426
   shapeName=dyn_cloth_bl_mid_03_capsule
   center=<5.9083 0.8395 0.0123>
   rotation=<rot 0 -102.2143 -0.0385>
  }]
 dyn_cloth_bl_mid_04 [{
   shapeType=Capsule
   length=10.338
   radius=6.4857
   shapeName=dyn_cloth_bl_mid_04_capsule
   center=<7.4271 -0.2902 -0.9641>
   rotation=<rot 0 -94.655 -3.1023>
  }]
 dyn_cloth_bl_mid_05 [{
   shapeType=Capsule
   length=11.059
   radius=5.4454
   shapeName=dyn_cloth_bl_mid_05_capsule
   center=<4.9686 0.6207 -0.354>
   rotation=<rot 0 -107.6433 -1.8249>
  }]
 dyn_cloth_bl_mid_07 [{shapeType=Capsule, length=11.755, radius=5.505, shapeName=dyn_cloth_bl_mid_07_capsule, center=<5.2643 0.0001 0>, rotation=<rot 0 -90 0>}]
 dyn_cloth_fl_01 [{
   shapeType=Capsule
   length=8.3586
   radius=3.7081
   shapeName=dyn_cloth_fl_01_capsule
   center=<7.2702 0.3041 -0.5334>
   rotation=<rot -0.0507 -90.1097 0.6233>
  }]
 dyn_cloth_fl_02 [{
   shapeType=Capsule
   length=8.8312
   radius=4.2946
   shapeName=dyn_cloth_fl_02_capsule
   center=<8.6021 0.7977 1.2207>
   rotation=<rot 0 90.7364 -5.937>
  }]
 dyn_cloth_fl_03 [{
   shapeType=Capsule
   length=5.7992
   radius=4.4245
   shapeName=dyn_cloth_fl_03_capsule
   center=<5.7888 0.5272 0.3479>
   rotation=<rot 0 84.6145 1.9513>
  }]
 dyn_cloth_fl_04 [{
   shapeType=Capsule
   length=5.4189
   radius=5.3814
   shapeName=dyn_cloth_fl_04_capsule
   center=<5.3487 3 0.382>
   rotation=<rot 0 87.2766 -5.1386>
  }]
 dyn_cloth_fl_05 [{
   shapeType=Capsule
   length=4.2272
   radius=6.3605
   shapeName=dyn_cloth_fl_05_capsule
   center=<4.5969 3 0.4199>
   rotation=<rot 0 86.9214 5.2466>
  }]
 dyn_cloth_fl_06 [{
   shapeType=Capsule
   length=3.7331
   radius=5.5
   shapeName=dyn_cloth_fl_06_capsule
   center=<4.9736 3.8744 0.7072>
   rotation=<rot -6.2657 -90 4.0639>
  }]
 dyn_cloth_fl_MSTR [{shapeType=Capsule, length=0.505, radius=0.505, rotation=<rot 0 -90 0>}]
 dyn_cloth_l_01 [{shapeType=Capsule, length=16.505, radius=4.255, shapeName=dyn_cloth_l_01_capsule, center=<3.2218 0.0001 0>, rotation=<rot 0 -90 0>}]
 dyn_cloth_l_02 [{
   shapeType=Capsule
   length=17.4768
   radius=5.0265
   shapeName=dyn_cloth_l_02_capsule
   center=<7.8153 0.4313 -0.1951>
   rotation=<rot 0.0001 -83.5187 -176.9043>
  }]
 dyn_cloth_l_03 [{
   shapeType=Capsule
   length=12.7768
   radius=4.8561
   shapeName=dyn_cloth_l_03_capsule
   center=<4.1117 3.4716 -2.911>
   rotation=<rot 1.2752 -88.0418 5.4315>
  }]
 dyn_cloth_l_04 [{
   shapeType=Capsule
   length=13.527
   radius=5.1435
   shapeName=dyn_cloth_l_04_capsule
   center=<7.2892 0.5919 -2.7263>
   rotation=<rot -0.0001 -85.9631 5.9553>
  }]
 dyn_cloth_l_05 [{
   shapeType=Capsule
   length=17.066
   radius=5.7786
   shapeName=dyn_cloth_l_05_capsule
   center=<6.2207 0.804 -3.176>
   rotation=<rot -3.1154 -89.7011 1.6759>
  }]
 dyn_cloth_l_06 [{
   shapeType=Capsule
   length=16.7872
   radius=5.8843
   shapeName=dyn_cloth_l_06_capsule
   center=<5.9471 -1.0866 -4.5768>
   rotation=<rot 4.9516 -90.0116 -0.0211>
  }]
 dyn_cloth_l_mid_02 [{
   shapeType=Capsule
   length=7.3251
   radius=6.781
   shapeName=dyn_cloth_l_mid_02_capsule
   center=<8.3767 4 -1.2081>
   rotation=<rot -8.6055 -89.9999 6.3266>
  }]
 dyn_cloth_l_mid_03 [{
   shapeType=Capsule
   length=7.2695
   radius=4.8288
   shapeName=dyn_cloth_l_mid_03_capsule
   center=<4.1802 4 -1.6985>
   rotation=<rot 2.8557 -83.6297 -4.9566>
  }]
 dyn_cloth_l_mid_04 [{
   shapeType=Capsule
   length=9.0499
   radius=5.1875
   shapeName=dyn_cloth_l_mid_04_capsule
   center=<6.2936 4 1.4393>
   rotation=<rot 0 -91.4713 0.8133>
  }]
 dyn_cloth_l_mid_05 [{
   shapeType=Capsule
   length=10.7975
   radius=5.25
   shapeName=dyn_cloth_l_mid_05_capsule
   center=<5.617 8 -0.2237>
   rotation=<rot 0 -91.5533 -0.3091>
  }]
 dyn_cloth_l_mid_07 [{
   shapeType=Capsule
   length=11.4098
   radius=6.4306
   shapeName=dyn_cloth_l_mid_07_capsule
   center=<6.7118 4 -0.4294>
   rotation=<rot 0.1121 -89.9999 0.1429>
  }]
 dyn_cloth_l_mid_08 [{
   shapeType=Capsule
   length=6.051
   radius=3.2032
   shapeName=dyn_cloth_l_mid_08_capsule
   center=<4.0162 0.7292 -1.8479>
   rotation=<rot 0 90.5334 8.3185>
  }]
 pelvis [{shapeType=Capsule, length=9.8119, radius=9.8283}]
 root [{shapeType=Box, extentX=240.0799, extentY=255.951, extentZ=19.18, center=<0 -0.0001 -7.3631>}]
 spine_01 [{shapeType=Capsule, length=2, radius=2}]
 spine_02 [{shapeType=Capsule, length=2, radius=2}]
 thigh_l [{shapeType=Capsule, length=18.6263, radius=12.4426, center=<-25.6753 -0.8523 4.1904>, rotation=<rot 86.1435 -177.9 -179.3222>}]
 thigh_r [{shapeType=Capsule, length=22.2951, radius=12.4426, center=<27.9634 0.9073 -4.1192>, rotation=<rot 86.1435 -177.9 -179.3222>}]
 thigh_twist_01_l [{shapeType=Capsule, length=2, radius=2}]
 thigh_twist_01_r [{shapeType=Capsule, length=2, radius=2}]
}
constraints [
 {bone1Name=dyn_cloth_l_02, bone2Name=dyn_cloth_l_01, swing1LimitDegrees=75, swing2LimitDegrees=75, twistLimitDegrees=15}
 {bone1Name=dyn_cloth_l_03, bone2Name=dyn_cloth_l_02, swing1LimitDegrees=75, swing2LimitDegrees=75, twistLimitDegrees=15}
 {bone1Name=dyn_cloth_l_04, bone2Name=dyn_cloth_l_03, swing1LimitDegrees=75, swing2LimitDegrees=75, twistLimitDegrees=15}
 {bone1Name=dyn_cloth_l_05, bone2Name=dyn_cloth_l_04, swing1LimitDegrees=75, swing2LimitDegrees=75, twistLimitDegrees=15}
 {bone1Name=dyn_cloth_l_06, bone2Name=dyn_cloth_l_05, swing1LimitDegrees=75, swing2LimitDegrees=75, twistLimitDegrees=15}
 {bone1Name=dyn_cloth_fl_01, bone2Name=dyn_cloth_fl_MSTR, swing1LimitDegrees=75, swing2LimitDegrees=75, twistLimitDegrees=15}
 {bone1Name=dyn_cloth_fl_02, bone2Name=dyn_cloth_fl_01, swing1LimitDegrees=75, swing2LimitDegrees=75, twistLimitDegrees=15}
 {bone1Name=dyn_cloth_fl_03, bone2Name=dyn_cloth_fl_02, swing1LimitDegrees=75, swing2LimitDegrees=75, twistLimitDegrees=15}
 {bone1Name=dyn_cloth_fl_04, bone2Name=dyn_cloth_fl_03, swing1LimitDegrees=75, swing2LimitDegrees=75, twistLimitDegrees=15}
 {bone1Name=dyn_cloth_fl_05, bone2Name=dyn_cloth_fl_04, swing1LimitDegrees=75, swing2LimitDegrees=75, twistLimitDegrees=15}
 {bone1Name=dyn_cloth_fl_06, bone2Name=dyn_cloth_fl_05, swing1LimitDegrees=75, swing2LimitDegrees=75, twistLimitDegrees=15}
 {bone1Name=dyn_cloth_bl_01, bone2Name=pelvis, swing1LimitDegrees=75, swing2LimitDegrees=75, twistLimitDegrees=15}
 {bone1Name=dyn_cloth_bl_02, bone2Name=dyn_cloth_bl_01, swing1LimitDegrees=75, swing2LimitDegrees=75, twistLimitDegrees=15}
 {bone1Name=dyn_cloth_bl_03, bone2Name=dyn_cloth_bl_02, swing1LimitDegrees=75, swing2LimitDegrees=75, twistLimitDegrees=15}
 {bone1Name=dyn_cloth_bl_04, bone2Name=dyn_cloth_bl_03, swing1LimitDegrees=75, swing2LimitDegrees=75, twistLimitDegrees=15}
 {bone1Name=dyn_cloth_bl_05, bone2Name=dyn_cloth_bl_04, swing1LimitDegrees=75, swing2LimitDegrees=75, twistLimitDegrees=15}
 {bone1Name=dyn_cloth_bl_mid_01, bone2Name=pelvis, swing1LimitDegrees=75, swing2LimitDegrees=75, twistLimitDegrees=15}
 {bone1Name=dyn_cloth_bl_mid_02, bone2Name=dyn_cloth_bl_mid_01, swing1LimitDegrees=75, swing2LimitDegrees=75, twistLimitDegrees=15}
 {bone1Name=dyn_cloth_bl_mid_03, bone2Name=dyn_cloth_bl_mid_02, swing1LimitDegrees=75, swing2LimitDegrees=75, twistLimitDegrees=15}
 {bone1Name=dyn_cloth_bl_mid_04, bone2Name=dyn_cloth_bl_mid_03, swing1LimitDegrees=75, swing2LimitDegrees=75, twistLimitDegrees=15}
 {bone1Name=dyn_cloth_bl_mid_05, bone2Name=dyn_cloth_bl_mid_04, swing1LimitDegrees=75, swing2LimitDegrees=75, twistLimitDegrees=15}
 {bone1Name=dyn_cloth_bl_mid_07, bone2Name=dyn_cloth_bl_mid_05, swing1LimitDegrees=75, swing2LimitDegrees=75, twistLimitDegrees=15}
 {bone1Name=dyn_cloth_l_mid_08, bone2Name=thigh_twist_01_l, swing1LimitDegrees=75, swing2LimitDegrees=75, twistLimitDegrees=15}
 {bone1Name=dyn_cloth_l_mid_02, bone2Name=dyn_cloth_l_mid_08, swing1LimitDegrees=75, swing2LimitDegrees=75, twistLimitDegrees=15}
 {bone1Name=dyn_cloth_l_mid_03, bone2Name=dyn_cloth_l_mid_02, swing1LimitDegrees=75, swing2LimitDegrees=75, twistLimitDegrees=15}
 {bone1Name=dyn_cloth_l_mid_04, bone2Name=dyn_cloth_l_mid_03, swing1LimitDegrees=75, swing2LimitDegrees=75, twistLimitDegrees=15}
 {bone1Name=dyn_cloth_l_mid_05, bone2Name=dyn_cloth_l_mid_04, swing1LimitDegrees=75, swing2LimitDegrees=75, twistLimitDegrees=15}
 {bone1Name=dyn_cloth_l_mid_07, bone2Name=dyn_cloth_l_mid_05, swing1LimitDegrees=75, swing2LimitDegrees=75, twistLimitDegrees=15}
 {
  bone1Name=thigh_l
  bone2Name=pelvis
  swing1LimitDegrees=45
  swing1Motion=Free
  swing2LimitDegrees=45
  swing2Motion=Free
  twistLimitDegrees=45
  twistMotion=Free
 }
 {
  bone1Name=calf_l
  bone2Name=thigh_l
  swing1LimitDegrees=45
  swing1Motion=Free
  swing2LimitDegrees=45
  swing2Motion=Free
  twistLimitDegrees=45
  twistMotion=Free
 }
 {
  bone1Name=thigh_r
  bone2Name=pelvis
  swing1LimitDegrees=45
  swing1Motion=Free
  swing2LimitDegrees=45
  swing2Motion=Free
  twistLimitDegrees=45
  twistMotion=Free
 }
 {
  bone1Name=calf_r
  bone2Name=thigh_r
  swing1LimitDegrees=45
  swing1Motion=Free
  swing2LimitDegrees=45
  swing2Motion=Free
  twistLimitDegrees=45
  twistMotion=Free
 }
 {bone1Name=dyn_cloth_l_01, bone2Name=thigh_twist_01_l, swing1LimitDegrees=75, swing2LimitDegrees=75, twistLimitDegrees=15}
 {
  bone1Name=spine_01
  bone2Name=pelvis
  swing1LimitDegrees=45
  swing1Motion=Free
  swing2LimitDegrees=45
  swing2Motion=Free
  twistLimitDegrees=45
  twistMotion=Free
 }
 {
  bone1Name=thigh_twist_01_l
  bone2Name=thigh_l
  swing1LimitDegrees=45
  swing1Motion=Free
  swing2LimitDegrees=45
  swing2Motion=Free
  twistLimitDegrees=45
  twistMotion=Free
 }
 {
  bone1Name=thigh_twist_01_r
  bone2Name=thigh_r
  swing1LimitDegrees=45
  swing1Motion=Free
  swing2LimitDegrees=45
  swing2Motion=Free
  twistLimitDegrees=45
  twistMotion=Free
 }
 {bone1Name=dyn_cloth_l_mid_05, bone2Name=dyn_cloth_fl_05, swing1LimitDegrees=45, swing2LimitDegrees=45, twistLimitDegrees=45}
 {bone1Name=dyn_cloth_l_mid_03, bone2Name=dyn_cloth_fl_03, swing1LimitDegrees=45, swing2LimitDegrees=45, twistLimitDegrees=45}
 {
  bone1Name=pelvis
  bone2Name=root
  swing1LimitDegrees=45
  swing1Motion=Free
  swing2LimitDegrees=45
  swing2Motion=Free
  twistLimitDegrees=45
  twistMotion=Free
 }
 {bone1Name=dyn_cloth_bl_mid_03, bone2Name=dyn_cloth_bl_03, swing1LimitDegrees=45, swing2LimitDegrees=45, twistLimitDegrees=45}
 {bone1Name=dyn_cloth_bl_mid_05, bone2Name=dyn_cloth_bl_05, swing1LimitDegrees=45, swing2LimitDegrees=45, twistLimitDegrees=45}
 {bone1Name=dyn_cloth_l_mid_03, bone2Name=dyn_cloth_l_03, swing1LimitDegrees=45, swing2LimitDegrees=45, twistLimitDegrees=45}
 {bone1Name=dyn_cloth_l_mid_05, bone2Name=dyn_cloth_l_05, swing1LimitDegrees=45, swing2LimitDegrees=45, twistLimitDegrees=45}
 {
  bone1Name=dyn_cloth_fl_MSTR
  bone2Name=thigh_twist_01_l
  swing1LimitDegrees=45
  swing1Motion=Free
  swing2LimitDegrees=45
  swing2Motion=Free
  twistLimitDegrees=45
  twistMotion=Free
 }
 {bone1Name=dyn_cloth_bl_mid_04, bone2Name=dyn_cloth_l_04, swing1LimitDegrees=45, swing2LimitDegrees=45, twistLimitDegrees=45}
 {
  bone1Name=spine_02
  bone2Name=spine_01
  swing1LimitDegrees=45
  swing1Motion=Free
  swing2LimitDegrees=45
  swing2Motion=Free
  twistLimitDegrees=45
  twistMotion=Free
 }
 {
  bone1Name=dyn_cloth_fl_05
  bone2Name=dyn_cloth_l_05
  swing1LimitDegrees=45
  swing1Motion=Free
  swing2LimitDegrees=45
  swing2Motion=Free
  twistLimitDegrees=45
  twistMotion=Free
 }
 {
  bone1Name=dyn_cloth_fl_06
  bone2Name=dyn_cloth_l_06
  swing1LimitDegrees=45
  swing1Motion=Free
  swing2LimitDegrees=45
  swing2Motion=Free
  twistLimitDegrees=45
  twistMotion=Free
 }
 {
  bone1Name=dyn_cloth_fl_04
  bone2Name=dyn_cloth_l_04
  swing1LimitDegrees=45
  swing1Motion=Free
  swing2LimitDegrees=45
  swing2Motion=Free
  twistLimitDegrees=45
  twistMotion=Free
 }
 {
  bone1Name=dyn_cloth_fl_03
  bone2Name=dyn_cloth_l_03
  swing1LimitDegrees=45
  swing1Motion=Free
  swing2LimitDegrees=45
  swing2Motion=Free
  twistLimitDegrees=45
  twistMotion=Free
 }
]
