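# Physics asset for the right-side dynamic skirt of the medium male Prisoner player body.
# Five capsule chains with limited swing and twist form the cloth: front and two side chains hang from the right thigh twist bone, two back-right chains from the pelvis.
# Cross-constraints tie neighboring chains at selected rows, approximating a cloth sheet with rigid bodies.
# Pelvis, spine and leg capsules, hip ring bodies and a flat root box complete the setup.
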PhysicsAsset /BRCosmetics/Characters/Player/Male/Medium/Bodies/M_MED_Prisoner/Meshes/M_MED_Prisoner_Dyn_Skirt_r
bodies {
 calf_l [{shapeType=Capsule, length=34.4819, radius=11.5365, center=<-28.4281 -1.6068 -0.8734>, rotation=<rot 86.7735 180 180>}]
 calf_r [{shapeType=Capsule, length=34.4819, radius=11.5365, center=<28.4281 1.6068 0.8734>, rotation=<rot 86.7735 180 180>}]
 dyn_cloth_br_01 [{shapeType=Capsule, length=11.255, radius=5.755, shapeName=dyn_cloth_br_01_capsule, center=<3.5336 0.0001 0>, rotation=<rot 0 -90 0>}]
 dyn_cloth_br_02 [{
   shapeType=Capsule
   length=10.6947
   radius=6.6214
   shapeName=dyn_cloth_br_02_capsule
   center=<8.0244 -0.1364 -0.3103>
   rotation=<rot 0 83.539 -0.0278>
  }]
 dyn_cloth_br_03 [{
   shapeType=Capsule
   length=6.5722
   radius=6.6465
   shapeName=dyn_cloth_br_03_capsule
   center=<5.5357 4 2.1763>
   rotation=<rot 0 90.8695 3.6163>
  }]
 dyn_cloth_br_04 [{
   shapeType=Capsule
   length=8.1834
   radius=6.8924
   shapeName=dyn_cloth_br_04_capsule
   center=<6.5082 4 2.7701>
   rotation=<rot -16.7031 -90 -2.1856>
  }]
 dyn_cloth_br_05 [{
   shapeType=Capsule
   length=8.6206
   radius=4.8739
   shapeName=dyn_cloth_br_05_capsule
   center=<5.2998 4 2.1706>
   rotation=<rot -17.0414 -89.9999 2.1982>
  }]
 dyn_cloth_br_mid_01 [{shapeType=Capsule, length=10.005, radius=5.005, shapeName=dyn_cloth_br_mid_01_capsule, center=<3.0347 -0.0002 0>, rotation=<rot 0 -90 0>}]
 dyn_cloth_br_mid_02 [{
   shapeType=Capsule
   length=8.0987
   radius=5.3436
   shapeName=dyn_cloth_br_mid_02_capsule
   center=<7.6413 0.58 -0.77>
   rotation=<rot 0 -75.1257 -1.1291>
  }]
 dyn_cloth_br_mid_03 [{
   shapeType=Capsule
   length=11.1437
   radius=6.0461
   shapeName=dyn_cloth_br_mid_03_capsule
   center=<6.8876 0.5413 0.4654>
   rotation=<rot 0 85.0107 -4.1587>
  }]
 dyn_cloth_br_mid_04 [{
   shapeType=Capsule
   length=10.7811
   radius=5.5036
   shapeName=dyn_cloth_br_mid_04_capsule
   center=<5.8753 4 -0.5989>
   rotation=<rot 0 90.5009 -5.0742>
  }]
 dyn_cloth_br_mid_05 [{shapeType=Capsule, length=16.005, radius=5.505, shapeName=dyn_cloth_br_mid_05_capsule, center=<3.0358 4 0>, rotation=<rot 0 -90 0>}]
 dyn_cloth_br_mid_07 [{shapeType=Capsule, length=11.755, radius=5.505, shapeName=dyn_cloth_br_mid_07_capsule, center=<4.9517 4 0>, rotation=<rot 0 -90 0>}]
 dyn_cloth_fr_01 [{
   shapeType=Capsule
   length=7.1921
   radius=5.8074
   shapeName=dyn_cloth_fr_01_capsule
   center=<5.4354 -0.3692 -0.0273>
   rotation=<rot 0 -84.1785 0.0883>
  }]
 dyn_cloth_fr_02 [{
   shapeType=Capsule
   length=5.9506
   radius=4.0711
   shapeName=dyn_cloth_fr_02_capsule
   center=<9.046 1.1887 -0.7912>
   rotation=<rot 0 -86.7128 -0.5769>
  }]
 dyn_cloth_fr_03 [{
   shapeType=Capsule
   length=5.9155
   radius=3.44
   shapeName=dyn_cloth_fr_03_capsule
   center=<5.9253 0.6156 -0.4902>
   rotation=<rot 0 -93.7547 -1.0605>
  }]
 dyn_cloth_fr_04 [{
   shapeType=Capsule
   length=5.4291
   radius=4.103
   shapeName=dyn_cloth_fr_04_capsule
   center=<5.5747 0.423 0.6108>
   rotation=<rot 0 -91.6013 -9.5973>
  }]
 dyn_cloth_fr_05 [{
   shapeType=Capsule
   length=6.2509
   radius=5.2716
   shapeName=dyn_cloth_fr_05_capsule
   center=<6.5934 4 0.3025>
   rotation=<rot 0 -99.4765 -0.1748>
  }]
 dyn_cloth_fr_06 [{
   shapeType=Capsule
   length=5.2874
   radius=5.6261
   shapeName=dyn_cloth_fr_06_capsule
   center=<5.6587 4 0.0417>
   rotation=<rot 0 -94.7291 -5.6379>
  }]
 dyn_cloth_fr_MSTR [{shapeType=Capsule, length=0.505, radius=0.505, rotation=<rot 0 -90 0>}]
 dyn_cloth_r_01 [{
   shapeType=Capsule
   length=11.5635
   radius=4.8319
   shapeName=dyn_cloth_r_01_capsule
   center=<2.8996 0.4059 0.5414>
   rotation=<rot 1.6349 -85.4853 4.4976>
  }]
 dyn_cloth_r_02 [{
   shapeType=Capsule
   length=8.8052
   radius=5.8535
   shapeName=dyn_cloth_r_02_capsule
   center=<8.723 0.2557 -1.4328>
   rotation=<rot -1.0304 96.885 -5.855>
  }]
 dyn_cloth_r_03 [{
   shapeType=Capsule
   length=8.7406
   radius=5.425
   shapeName=dyn_cloth_r_03_capsule
   center=<5.4773 2.7704 -3.4443>
   rotation=<rot 0 -95.0003 3.3563>
  }]
 dyn_cloth_r_04 [{
   shapeType=Capsule
   length=12.9886
   radius=5.388
   shapeName=dyn_cloth_r_04_capsule
   center=<5.7705 0.188 -1.9069>
   rotation=<rot -0.2995 86.7518 -1.6985>
  }]
 dyn_cloth_r_05 [{
   shapeType=Capsule
   length=10.613
   radius=4.2583
   shapeName=dyn_cloth_r_05_capsule
   center=<5.2755 4 -4.7414>
   rotation=<rot 1.7179 -176.5093 -4.7272>
  }]
 dyn_cloth_r_06 [{
   shapeType=Capsule
   length=15.7376
   radius=5.4884
   shapeName=dyn_cloth_r_06_capsule
   center=<5.407 4 -1.1595>
   rotation=<rot 11.8774 -89.9999 -2.8201>
  }]
 dyn_cloth_r_mid_02 [{
   shapeType=Capsule
   length=6.6081
   radius=4.8629
   shapeName=dyn_cloth_r_mid_02_capsule
   center=<8.9185 0.0005 -0.8483>
   rotation=<rot -3.2326 85.8893 -8.9278>
  }]
 dyn_cloth_r_mid_03 [{
   shapeType=Capsule
   length=5.5525
   radius=3.6153
   shapeName=dyn_cloth_r_mid_03_capsule
   center=<5.8609 3.0298 -4.3776>
   rotation=<rot 10.0001 -96.1744 -0.1359>
  }]
 dyn_cloth_r_mid_04 [{
   shapeType=Capsule
   length=7.4157
   radius=7.3261
   shapeName=dyn_cloth_r_mid_04_capsule
   center=<6.3233 4 -1.1689>
   rotation=<rot 10.1727 -89.9999 -3.7777>
  }]
 dyn_cloth_r_mid_05 [{
   shapeType=Capsule
   length=8.1033
   radius=4.7134
   shapeName=dyn_cloth_r_mid_05_capsule
   center=<4.1864 4 -1.3539>
   rotation=<rot -0.0001 92.4997 -3.102>
  }]
 dyn_cloth_r_mid_07 [{shapeType=Capsule, length=8.755, radius=5.005, shapeName=dyn_cloth_r_mid_07_capsule, center=<7.1736 4 -0.9892>, rotation=<rot 0 -90 0>}]
 dyn_cloth_r_mid_09 [{
   shapeType=Capsule
   length=7.1828
   radius=4.2275
   shapeName=dyn_cloth_r_mid_09_capsule
   center=<4.3232 0.628 -1.7819>
   rotation=<rot 0 -88.5491 -178.9527>
  }]
 dyn_hip_ring [{shapeType=Capsule, length=3.4181, radius=2.2544, center=<4.8168 0 -0.0002>, rotation=<rot 90 0 0>}]
 dyn_hip_ring_MSTR [{shapeType=Capsule, length=0.505, radius=0.505, rotation=<rot 0 -90 0>}]
 dyn_side_ring_MSTR [{shapeType=Capsule, length=0.505, radius=0.505, rotation=<rot 0 -90 0>}]
 pelvis [{shapeType=Capsule, length=9.8119, radius=9.8283}]
 root [{shapeType=Box, extentX=240.0799, extentY=255.951, extentZ=19.18, center=<0 -0.0001 -7.3631>}]
 spine_01 [{shapeType=Capsule, length=2, radius=2}]
 spine_02 [{shapeType=Capsule, length=2, radius=2}]
 thigh_l [{shapeType=Capsule, length=18.6263, radius=12.4426, center=<-25.6753 -0.8523 4.1904>, rotation=<rot 86.1435 -177.9 -179.3222>}]
 thigh_r [{shapeType=Capsule, length=22.2951, radius=12.4426, center=<27.9634 0.9073 -4.1192>, rotation=<rot 86.1435 -177.9 -179.3222>}]
 thigh_twist_01_l [{shapeType=Capsule, length=2, radius=2}]
 thigh_twist_01_r [{shapeType=Capsule, length=2, radius=2}]
}
constraints [
 {bone1Name=dyn_cloth_fr_01, bone2Name=dyn_cloth_fr_MSTR, swing1LimitDegrees=75, swing2LimitDegrees=75, twistLimitDegrees=15}
 {bone1Name=dyn_cloth_fr_02, bone2Name=dyn_cloth_fr_01, swing1LimitDegrees=75, swing2LimitDegrees=75, twistLimitDegrees=15}
 {bone1Name=dyn_cloth_fr_03, bone2Name=dyn_cloth_fr_02, swing1LimitDegrees=75, swing2LimitDegrees=75, twistLimitDegrees=15}
 {bone1Name=dyn_cloth_fr_04, bone2Name=dyn_cloth_fr_03, swing1LimitDegrees=75, swing2LimitDegrees=75, twistLimitDegrees=15}
 {bone1Name=dyn_cloth_fr_05, bone2Name=dyn_cloth_fr_04, swing1LimitDegrees=75, swing2LimitDegrees=75, twistLimitDegrees=15}
 {bone1Name=dyn_cloth_fr_06, bone2Name=dyn_cloth_fr_05, swing1LimitDegrees=75, swing2LimitDegrees=75, twistLimitDegrees=15}
 {bone1Name=dyn_cloth_r_mid_09, bone2Name=thigh_twist_01_r, swing1LimitDegrees=75, swing2LimitDegrees=75, twistLimitDegrees=15}
 {bone1Name=dyn_cloth_r_mid_02, bone2Name=dyn_cloth_r_mid_09, swing1LimitDegrees=75, swing2LimitDegrees=75, twistLimitDegrees=15}
 {bone1Name=dyn_cloth_r_mid_03, bone2Name=dyn_cloth_r_mid_02, swing1LimitDegrees=75, swing2LimitDegrees=75, twistLimitDegrees=15}
 {bone1Name=dyn_cloth_r_mid_04, bone2Name=dyn_cloth_r_mid_03, swing1LimitDegrees=75, swing2LimitDegrees=75, twistLimitDegrees=15}
 {bone1Name=dyn_cloth_r_mid_05, bone2Name=dyn_cloth_r_mid_04, swing1LimitDegrees=75, swing2LimitDegrees=75, twistLimitDegrees=15}
 {bone1Name=dyn_cloth_r_mid_07, bone2Name=dyn_cloth_r_mid_05, swing1LimitDegrees=75, swing2LimitDegrees=75, twistLimitDegrees=15}
 {bone1Name=dyn_cloth_r_01, bone2Name=thigh_twist_01_r, swing1LimitDegrees=75, swing2LimitDegrees=75, twistLimitDegrees=15}
 {bone1Name=dyn_cloth_r_02, bone2Name=dyn_cloth_r_01, swing1LimitDegrees=75, swing2LimitDegrees=75, twistLimitDegrees=15}
 {bone1Name=dyn_cloth_r_03, bone2Name=dyn_cloth_r_02, swing1LimitDegrees=75, swing2LimitDegrees=75, twistLimitDegrees=15}
 {bone1Name=dyn_cloth_r_04, bone2Name=dyn_cloth_r_03, swing1LimitDegrees=75, swing2LimitDegrees=75, twistLimitDegrees=15}
 {bone1Name=dyn_cloth_r_05, bone2Name=dyn_cloth_r_04, swing1LimitDegrees=75, swing2LimitDegrees=75, twistLimitDegrees=15}
 {bone1Name=dyn_cloth_r_06, bone2Name=dyn_cloth_r_05, swing1LimitDegrees=75, swing2LimitDegrees=75, twistLimitDegrees=15}
 {bone1Name=dyn_cloth_br_mid_01, bone2Name=pelvis, swing1LimitDegrees=75, swing2LimitDegrees=75, twistLimitDegrees=15}
 {bone1Name=dyn_cloth_br_mid_02, bone2Name=dyn_cloth_br_mid_01, swing1LimitDegrees=75, swing2LimitDegrees=75, twistLimitDegrees=15}
 {bone1Name=dyn_cloth_br_mid_03, bone2Name=dyn_cloth_br_mid_02, swing1LimitDegrees=75, swing2LimitDegrees=75, twistLimitDegrees=15}
 {bone1Name=dyn_cloth_br_mid_04, bone2Name=dyn_cloth_br_mid_03, swing1LimitDegrees=75, swing2LimitDegrees=75, twistLimitDegrees=15}
 {bone1Name=dyn_cloth_br_mid_05, bone2Name=dyn_cloth_br_mid_04, swing1LimitDegrees=75, swing2LimitDegrees=75, twistLimitDegrees=15}
 {bone1Name=dyn_cloth_br_mid_07, bone2Name=dyn_cloth_br_mid_05, swing1LimitDegrees=75, swing2LimitDegrees=75, twistLimitDegrees=15}
 {bone1Name=dyn_cloth_br_01, bone2Name=pelvis, swing1LimitDegrees=75, swing2LimitDegrees=75, twistLimitDegrees=15}
 {bone1Name=dyn_cloth_br_02, bone2Name=dyn_cloth_br_01, swing1LimitDegrees=75, swing2LimitDegrees=75, twistLimitDegrees=15}
 {bone1Name=dyn_cloth_br_03, bone2Name=dyn_cloth_br_02, swing1LimitDegrees=75, swing2LimitDegrees=75, twistLimitDegrees=15}
 {bone1Name=dyn_cloth_br_04, bone2Name=dyn_cloth_br_03, swing1LimitDegrees=75, swing2LimitDegrees=75, twistLimitDegrees=15}
 {bone1Name=dyn_cloth_br_05, bone2Name=dyn_cloth_br_04, swing1LimitDegrees=75, swing2LimitDegrees=75, twistLimitDegrees=15}
 {
  bone1Name=thigh_l
  bone2Name=pelvis
  swing1LimitDegrees=45
  swing1Motion=Free
  swing2LimitDegrees=45
  swing2Motion=Free
  twistLimitDegrees=45
  twistMotion=Free
 }
 {
  bone1Name=calf_l
  bone2Name=thigh_l
  swing1LimitDegrees=45
  swing1Motion=Free
  swing2LimitDegrees=45
  swing2Motion=Free
  twistLimitDegrees=45
  twistMotion=Free
 }
 {
  bone1Name=thigh_r
  bone2Name=pelvis
  swing1LimitDegrees=45
  swing1Motion=Free
  swing2LimitDegrees=45
  swing2Motion=Free
  twistLimitDegrees=45
  twistMotion=Free
 }
 {
  bone1Name=calf_r
  bone2Name=thigh_r
  swing1LimitDegrees=45
  swing1Motion=Free
  swing2LimitDegrees=45
  swing2Motion=Free
  twistLimitDegrees=45
  twistMotion=Free
 }
 {
  bone1Name=spine_01
  bone2Name=pelvis
  swing1LimitDegrees=45
  swing1Motion=Free
  swing2LimitDegrees=45
  swing2Motion=Free
  twistLimitDegrees=45
  twistMotion=Free
 }
 {
  bone1Name=thigh_twist_01_l
  bone2Name=thigh_l
  swing1LimitDegrees=45
  swing1Motion=Free
  swing2LimitDegrees=45
  swing2Motion=Free
  twistLimitDegrees=45
  twistMotion=Free
 }
 {
  bone1Name=thigh_twist_01_r
  bone2Name=thigh_r
  swing1LimitDegrees=45
  swing1Motion=Free
  swing2LimitDegrees=45
  swing2Motion=Free
  twistLimitDegrees=45
  twistMotion=Free
 }
 {
  bone1Name=pelvis
  bone2Name=root
  swing1LimitDegrees=45
  swing1Motion=Free
  swing2LimitDegrees=45
  swing2Motion=Free
  twistLimitDegrees=45
  twistMotion=Free
 }
 {bone1Name=dyn_cloth_br_mid_03, bone2Name=dyn_cloth_br_03, swing1LimitDegrees=45, swing2LimitDegrees=45, twistLimitDegrees=45}
 {bone1Name=dyn_cloth_br_mid_05, bone2Name=dyn_cloth_br_05, swing1LimitDegrees=45, swing2LimitDegrees=45, twistLimitDegrees=45}
 {bone1Name=dyn_cloth_r_mid_03, bone2Name=dyn_cloth_fr_03, swing1LimitDegrees=45, swing2LimitDegrees=45, twistLimitDegrees=45}
 {bone1Name=dyn_cloth_r_mid_05, bone2Name=dyn_cloth_fr_05, swing1LimitDegrees=45, swing2LimitDegrees=45, twistLimitDegrees=45}
 {bone1Name=dyn_cloth_br_mid_05, bone2Name=dyn_cloth_r_05, swing1LimitDegrees=45, swing2LimitDegrees=45, twistLimitDegrees=45}
 {bone1Name=dyn_cloth_br_mid_03, bone2Name=dyn_cloth_r_03, swing1LimitDegrees=45, swing2LimitDegrees=45, twistLimitDegrees=45}
 {
  bone1Name=dyn_cloth_fr_MSTR
  bone2Name=thigh_twist_01_r
  swing1LimitDegrees=45
  swing1Motion=Free
  swing2LimitDegrees=45
  swing2Motion=Free
  twistLimitDegrees=45
  twistMotion=Free
 }
 {bone1Name=dyn_cloth_br_mid_04, bone2Name=dyn_cloth_r_04, swing1LimitDegrees=45, swing2LimitDegrees=45, twistLimitDegrees=45}
 {bone1Name=dyn_hip_ring, bone2Name=pelvis, swing1LimitDegrees=45, swing2LimitDegrees=45, twistLimitDegrees=15}
 {
  bone1Name=spine_02
  bone2Name=spine_01
  swing1LimitDegrees=45
  swing1Motion=Free
  swing2LimitDegrees=45
  swing2Motion=Free
  twistLimitDegrees=45
  twistMotion=Free
 }
 {
  bone1Name=dyn_hip_ring_MSTR
  bone2Name=pelvis
  swing1LimitDegrees=45
  swing1Motion=Free
  swing2LimitDegrees=45
  swing2Motion=Free
  twistLimitDegrees=45
  twistMotion=Free
 }
 {
  bone1Name=dyn_cloth_fr_03
  bone2Name=dyn_cloth_r_03
  swing1LimitDegrees=45
  swing1Motion=Free
  swing2LimitDegrees=45
  swing2Motion=Free
  twistLimitDegrees=45
  twistMotion=Free
 }
 {
  bone1Name=dyn_cloth_fr_02
  bone2Name=dyn_cloth_r_02
  swing1LimitDegrees=45
  swing1Motion=Free
  swing2LimitDegrees=45
  swing2Motion=Free
  twistLimitDegrees=45
  twistMotion=Free
 }
 {
  bone1Name=dyn_cloth_fr_01
  bone2Name=dyn_cloth_r_01
  swing1LimitDegrees=45
  swing1Motion=Free
  swing2LimitDegrees=45
  swing2Motion=Free
  twistLimitDegrees=45
  twistMotion=Free
 }
]
