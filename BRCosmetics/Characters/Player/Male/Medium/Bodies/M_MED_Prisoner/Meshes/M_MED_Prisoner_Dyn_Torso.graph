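# Physics asset for the medium male Prisoner character's dynamic torso, covering four-link secondary chains on both forearms and the neck.
# Forearm chains hang from the first twist bones and the neck chain from neck_02, each link constrained to the previous with tight twist limits.
# Right-arm links allow twice the swing of left-arm and neck links.
# Pelvis, spine, neck, forearm and hand bodies use free angular constraints, providing anchors and collision shapes.
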PhysicsAsset /BRCosmetics/Characters/Player/Male/Medium/Bodies/M_MED_Prisoner/Meshes/M_MED_Prisoner_Dyn_Torso
bodies {
 hand_l [{shapeType=Capsule, length=6.5391, radius=5.4527, center=<5.8709 -1.1762 -0.22>, rotation=<rot 81.3173 179.9999 179.9999>}]
 hand_r [{shapeType=Capsule, length=6.5391, radius=5.4527, center=<-5.8709 1.1762 0.22>, rotation=<rot 81.3173 179.9999 179.9999>}]
 lowerarm_l [{shapeType=Capsule, length=2, radius=2}]
 lowerarm_l_chain_01 [{
   shapeType=Capsule
   length=3.0021
   radius=2.2074
   shapeName=lowerarm_l_chain_01_capsule
   center=<2.4421 0.077 -0.1334>
   rotation=<rot -0.3021 -88.8288 -88.2865>
  }]
 lowerarm_l_chain_02 [{
   shapeType=Capsule
   length=3.0021
   radius=1.9841
   shapeName=lowerarm_l_chain_02_capsule
   center=<2.2991 0.2004 -0.2742>
   rotation=<rot -0.3394 -97.253 -91.9251>
  }]
 lowerarm_l_chain_03 [{
   shapeType=Capsule
   length=3.0021
   radius=2.0964
   shapeName=lowerarm_l_chain_03_capsule
   center=<2.3723 0.2178 -0.1849>
   rotation=<rot 0.3099 86.7327 -88.2424>
  }]
 lowerarm_l_chain_04 [{
   shapeType=Capsule
   length=3.7928
   radius=1.9605
   shapeName=lowerarm_l_chain_04_capsule
   center=<2.6946 0.1404 -0.4886>
   rotation=<rot 0 -89.6892 -87.3118>
  }]
 lowerarm_r [{shapeType=Capsule, length=2, radius=2}]
 lowerarm_r_chain_01 [{
   shapeType=Capsule
   length=3.0021
   radius=2.1454
   shapeName=lowerarm_r_chain_01_capsule
   center=<2.443 0.0279 0.1525>
   rotation=<rot 0 -98.8531 -88.4134>
  }]
 lowerarm_r_chain_02 [{
   shapeType=Capsule
   length=3.0021
   radius=2.2118
   shapeName=lowerarm_r_chain_02_capsule
   center=<2.2998 0.0886 0.3238>
   rotation=<rot 0 89.5962 -91.5123>
  }]
 lowerarm_r_chain_03 [{
   shapeType=Capsule
   length=3.0021
   radius=2.219
   shapeName=lowerarm_r_chain_03_capsule
   center=<2.3729 0.1307 0.2501>
   rotation=<rot 0 96.9152 -89.2142>
  }]
 lowerarm_r_chain_04 [{
   shapeType=Capsule
   length=3.7928
   radius=2.1299
   shapeName=lowerarm_r_chain_04_capsule
   center=<2.6943 -0.0185 0.519>
   rotation=<rot 0 89.3079 -87.3845>
  }]
 lowerarm_twist_01_l [{shapeType=Capsule, length=2, radius=2}]
 lowerarm_twist_01_r [{shapeType=Capsule, length=2, radius=2}]
 lowerarm_twist_02_l [{shapeType=Capsule, length=2, radius=2}]
 lowerarm_twist_02_r [{shapeType=Capsule, length=2, radius=2}]
 neck_01 [{shapeType=Capsule, length=2, radius=2}]
 neck_02 [{shapeType=Capsule, length=2, radius=2}]
 neck_chain_01 [{shapeType=Capsule, length=2, radius=2, center=<3 0 0>, rotation=<rot 90 0 0>}]
 neck_chain_02 [{shapeType=Capsule, length=2, radius=2, center=<3 0 0>, rotation=<rot 90 0 0>}]
 neck_chain_03 [{shapeType=Capsule, length=2, radius=2, center=<3 0 0>, rotation=<rot 90 0 0>}]
 neck_chain_04 [{shapeType=Capsule, length=2, radius=2, center=<3 0 0>, rotation=<rot 90 0 0>}]
 pelvis [{shapeType=Capsule, length=12.2719, radius=9.9118}]
 pinky_metacarpal_l [{shapeType=Capsule, length=2, radius=2}]
 pinky_metacarpal_r [{shapeType=Capsule, length=2, radius=2}]
 spine_04 [{shapeType=Capsule, length=9.136, radius=12.9949, center=<1.6387 0.8439 0>}]
 spine_05 [
  {shapeType=Capsule, length=5.5974, radius=13.4524, center=<-3.7459 -0.1971 6.2518>, rotation=<rot 90 -90 -70.9537>}
  {shapeType=Capsule, length=5.5974, radius=13.4524, center=<-3.7459 -0.1971 -6.2522>, rotation=<rot 90 14.0363 33.0826>}
 ]
}
constraints [
 {bone1Name=lowerarm_l_chain_01, bone2Name=lowerarm_twist_01_l, swing1LimitDegrees=45, swing2LimitDegrees=45, twistLimitDegrees=15}
 {bone1Name=lowerarm_l_chain_02, bone2Name=lowerarm_l_chain_01, swing1LimitDegrees=45, swing2LimitDegrees=45, twistLimitDegrees=15}
 {bone1Name=lowerarm_l_chain_03, bone2Name=lowerarm_l_chain_02, swing1LimitDegrees=45, swing2LimitDegrees=45, twistLimitDegrees=15}
 {bone1Name=lowerarm_l_chain_04, bone2Name=lowerarm_l_chain_03, swing1LimitDegrees=45, swing2LimitDegrees=45, twistLimitDegrees=15}
 {bone1Name=lowerarm_r_chain_01, bone2Name=lowerarm_twist_01_r, swing1LimitDegrees=90, swing2LimitDegrees=90, twistLimitDegrees=15}
 {bone1Name=lowerarm_r_chain_02, bone2Name=lowerarm_r_chain_01, swing1LimitDegrees=90, swing2LimitDegrees=90, twistLimitDegrees=15}
 {bone1Name=lowerarm_r_chain_03, bone2Name=lowerarm_r_chain_02, swing1LimitDegrees=90, swing2LimitDegrees=90, twistLimitDegrees=15}
 {bone1Name=lowerarm_r_chain_04, bone2Name=lowerarm_r_chain_03, swing1LimitDegrees=90, swing2LimitDegrees=90, twistLimitDegrees=15}
 {
  bone1Name=lowerarm_twist_01_r
  bone2Name=lowerarm_r
  swing1LimitDegrees=45
  swing1Motion=Free
  swing2LimitDegrees=45
  swing2Motion=Free
  twistLimitDegrees=45
  twistMotion=Free
 }
 {
  bone1Name=lowerarm_twist_01_l
  bone2Name=lowerarm_l
  swing1LimitDegrees=45
  swing1Motion=Free
  swing2LimitDegrees=45
  swing2Motion=Free
  twistLimitDegrees=45
  twistMotion=Free
 }
 {
  bone1Name=spine_05
  bone2Name=spine_04
  swing1LimitDegrees=45
  swing1Motion=Free
  swing2LimitDegrees=45
  swing2Motion=Free
  twistLimitDegrees=45
  twistMotion=Free
 }
 {
  bone1Name=spine_04
  bone2Name=pelvis
  swing1LimitDegrees=45
  swing1Motion=Free
  swing2LimitDegrees=45
  swing2Motion=Free
  twistLimitDegrees=45
  twistMotion=Free
 }
 {
  bone1Name=hand_l
  bone2Name=lowerarm_l
  swing1LimitDegrees=45
  swing1Motion=Free
  swing2LimitDegrees=45
  swing2Motion=Free
  twistLimitDegrees=45
  twistMotion=Free
 }
 {
  bone1Name=hand_r
  bone2Name=lowerarm_r
  swing1LimitDegrees=45
  swing1Motion=Free
  swing2LimitDegrees=45
  swing2Motion=Free
  twistLimitDegrees=45
  twistMotion=Free
 }
 {
  bone1Name=neck_01
  bone2Name=spine_05
  swing1LimitDegrees=45
  swing1Motion=Free
  swing2LimitDegrees=45
  swing2Motion=Free
  twistLimitDegrees=45
  twistMotion=Free
 }
 {
  bone1Name=neck_02
  bone2Name=neck_01
  swing1LimitDegrees=45
  swing1Motion=Free
  swing2LimitDegrees=45
  swing2Motion=Free
  twistLimitDegrees=45
  twistMotion=Free
 }
 {
  bone1Name=lowerarm_l
  bone2Name=spine_05
  swing1LimitDegrees=45
  swing1Motion=Free
  swing2LimitDegrees=45
  swing2Motion=Free
  twistLimitDegrees=45
  twistMotion=Free
 }
 {
  bone1Name=pinky_metacarpal_l
  bone2Name=hand_l
  swing1LimitDegrees=45
  swing1Motion=Free
  swing2LimitDegrees=45
  swing2Motion=Free
  twistLimitDegrees=45
  twistMotion=Free
 }
 {
  bone1Name=lowerarm_twist_02_l
  bone2Name=lowerarm_l
  swing1LimitDegrees=45
  swing1Motion=Free
  swing2LimitDegrees=45
  swing2Motion=Free
  twistLimitDegrees=45
  twistMotion=Free
 }
 {
  bone1Name=lowerarm_r
  bone2Name=spine_05
  swing1LimitDegrees=45
  swing1Motion=Free
  swing2LimitDegrees=45
  swing2Motion=Free
  twistLimitDegrees=45
  twistMotion=Free
 }
 {
  bone1Name=pinky_metacarpal_r
  bone2Name=hand_r
  swing1LimitDegrees=45
  swing1Motion=Free
  swing2LimitDegrees=45
  swing2Motion=Free
  twistLimitDegrees=45
  twistMotion=Free
 }
 {
  bone1Name=lowerarm_twist_02_r
  bone2Name=lowerarm_r
  swing1LimitDegrees=45
  swing1Motion=Free
  swing2LimitDegrees=45
  swing2Motion=Free
  twistLimitDegrees=45
  twistMotion=Free
 }
 {bone1Name=neck_chain_01, bone2Name=neck_02, swing1LimitDegrees=45, swing2LimitDegrees=45, twistLimitDegrees=15}
 {bone1Name=neck_chain_02, bone2Name=neck_chain_01, swing1LimitDegrees=45, swing2LimitDegrees=45, twistLimitDegrees=15}
 {bone1Name=neck_chain_03, bone2Name=neck_chain_02, swing1LimitDegrees=45, swing2LimitDegrees=45, twistLimitDegrees=15}
 {bone1Name=neck_chain_04, bone2Name=neck_chain_03, swing1LimitDegrees=45, swing2LimitDegrees=45, twistLimitDegrees=15}
]
