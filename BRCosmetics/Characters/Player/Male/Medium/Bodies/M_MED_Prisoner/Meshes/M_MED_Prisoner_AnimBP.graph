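# Post-process style Animation Blueprint for the male Prisoner body, adding secondary motion to skirt cloth, chains and shoulder rocks.
# It copies the parent mesh pose, applies weighted bone constraints, then caches it as 'prerigid'.
# Cached-pose branches add Transform and Look At cloth correctives driven by thigh rotation, crouch-run and sprint values, plus RigidBody skirt and torso simulation with skydiving force.
# AnimDynamics swings neck and forearm chains from 'post rigid' before output.
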AnimBlueprint /BRCosmetics/Characters/Player/Male/Medium/Bodies/M_MED_Prisoner/Meshes/M_MED_Prisoner_AnimBP
target_skeleton=&/M_MED_Prisoner_Skeleton
graphs [
 AnimGraph {nodes [
   Anim Dynamics - Bone: lowerarm_r_chain_01(Alpha=1, ExternalForce="0,0,0") {
    angularDampingOverride=0.9999
    angularSpringConstant=10
    bAngularSpring=true
    bGravityOverrideInSimSpace=true
    bOverrideAngularDamping=true
    bUseGravityOverride=true
    alphaBoolBlend {blendOption=Linear}
    boundBone {boneName=lowerarm_r_chain_01}
    componentLinearAccScale=<1 1 1>
    componentLinearVelScale=<1 1 1>
    gravityOverride {z=-980}
    physicsBodyDefinitions [{
      sphereCollisionRadius=0
      boundBone {boneName=lowerarm_r_chain_01}
      boxExtents=<10 20 20>
      constraintSetup {angularLimitsMax=<20 100 100>, angularLimitsMin=<-20 -100 -100>, angularTarget=<0 0 0>}
      localJointOffset=<5 0 0>
     }]
   }
   "Constraint - neck_chain_01 (TTP)"!(ComponentPose, Alpha=1, ConstraintWeights_0=0.48, ConstraintWeights_1:float=0.3612, ConstraintWeights_2:float=0.1589) -> (Pose) {
    alphaBoolBlend {blendOption=Linear}
    boneToModify {boneName=neck_chain_01}
    constraintSetup [
     {transformType=Translation, targetBone {boneName=neck_02}}
     {transformType=Translation, targetBone {boneName=neck_01}}
     {targetBone {boneName=spine_05}}
    ]
    constraintWeights [1,1,1]
   }
   "Constraint - dyn_side_ring_MSTR (PPP)"!(ComponentPose, Alpha=1, ConstraintWeights_0=0.1922, ConstraintWeights_1:float=0.0749, ConstraintWeights_2:float=0.7329) -> (Pose) {
    alphaBoolBlend {blendOption=Linear}
    boneToModify {boneName=dyn_side_ring_MSTR}
    constraintSetup [{targetBone {boneName=spine_01}}, {targetBone {boneName=spine_02}}, {targetBone {boneName=pelvis}}]
    constraintWeights [1,1,1]
   }
   Copy Pose From Mesh() {bUseAttachedParent=true}
   Look At - Bone: dyn_cloth_bl_mid_01(Alpha=1) {
    alphaBoolBlend {blendOption=Linear}
    alphaScaleBiasClamp {bInterpResult=true}
    boneToModify {boneName=dyn_cloth_bl_mid_01}
    lookAtLocation=<0 -20 -10>
    lookAtTarget {boneReference {boneName=calf_l}}
    lookAt_Axis {axis=<1 0>}
    lookUp_Axis {axis {x=0}}
   }
   n1 = Get isNotCrouchRunning() -> (isNotCrouchRunning:double)
   "Transform (Modify) Bone - Bone: dyn_jacketBack_03_l"(Alpha=@n1, Translation="0,0,0", Rotation="25,0,0.000", Scale="1,1,1") {
    rotationMode=BMM_Additive
    rotationSpace=BCS_BoneSpace
    alphaBoolBlend {blendOption=Linear}
    alphaScaleBiasClamp {bInterpResult=true, interpSpeedDecreasing=2, interpSpeedIncreasing=5}
    boneToModify {boneName=dyn_jacketBack_03_l}
    rotation {pitch=25}
   }
   "Transform (Modify) Bone - Bone: dyn_cloth_fl_03"(Alpha=1, Translation="0,0,0", Rotation="5,-25,0.000", Scale="1,1,1") {
    rotationMode=BMM_Additive
    rotationSpace=BCS_BoneSpace
    translationMode=BMM_Additive
    translationSpace=BCS_BoneSpace
    alphaBoolBlend {blendOption=Linear}
    boneToModify {boneName=dyn_cloth_fl_03}
    rotation {pitch=5, yaw=-25}
   }
   n2 = Anim Dynamics - Bone: neck_chain_01(ComponentPose=Local To Component(LocalPose=Use cached pose 'post rigid'()), Alpha=1, ExternalForce="0,0,0") {
    angularDampingOverride=0.9999
    angularSpringConstant=0.5
    bGravityOverrideInSimSpace=true
    bOverrideAngularDamping=true
    bUseGravityOverride=true
    alphaBoolBlend {blendOption=Linear}
    boundBone {boneName=neck_chain_01}
    componentLinearAccScale=<1 1 1>
    componentLinearVelScale=<1 1 1>
    gravityOverride {z=-980}
    physicsBodyDefinitions [{
      sphereCollisionRadius=0
      boundBone {boneName=neck_chain_01}
      boxExtents=<10 20 20>
      constraintSetup {angularLimitsMax=<10 0 -20>, angularLimitsMin=<-10 -30 -50>, angularTarget=<0 0 0>}
      localJointOffset=<5 0 0>
     }]
   }
   n3 = Anim Dynamics - Bone: neck_chain_02(ComponentPose=@n2, Alpha=1, ExternalForce="0,0,0") {
    angularDampingOverride=0.9999
    angularSpringConstant=0.5
    bGravityOverrideInSimSpace=true
    bOverrideAngularDamping=true
    bUseGravityOverride=true
    alphaBoolBlend {blendOption=Linear}
    boundBone {boneName=neck_chain_02}
    componentLinearAccScale=<1 1 1>
    componentLinearVelScale=<1 1 1>
    gravityOverride {z=-980}
    physicsBodyDefinitions [{
      sphereCollisionRadius=0
      boundBone {boneName=neck_chain_02}
      boxExtents=<10 20 20>
      constraintSetup {angularLimitsMax=<10 60 10>, angularLimitsMin=<-10 -80 -50>, angularTarget=<0 0 0>}
      localJointOffset=<5 0 0>
     }]
   }
   n4 = Anim Dynamics - Bone: neck_chain_03(ComponentPose=@n3, Alpha=1, ExternalForce="0,0,0") {
    angularDampingOverride=0.999
    angularSpringConstant=0.5
    bGravityOverrideInSimSpace=true
    bOverrideAngularDamping=true
    bUseGravityOverride=true
    alphaBoolBlend {blendOption=Linear}
    boundBone {boneName=neck_chain_03}
    componentLinearAccScale=<1 1 1>
    componentLinearVelScale=<1 1 1>
    gravityOverride {z=-980}
    physicsBodyDefinitions [{
      sphereCollisionRadius=0
      boundBone {boneName=neck_chain_03}
      boxExtents=<10 20 20>
      constraintSetup {angularLimitsMax=<10 100 25>, angularLimitsMin=<-10 -100 -50>, angularTarget=<0 0 0>}
      localJointOffset=<5 0 0>
     }]
   }
   n5 = Anim Dynamics - Bone: neck_chain_04(ComponentPose=@n4, Alpha=1, ExternalForce="0,0,0") {
    angularDampingOverride=0.999
    angularSpringConstant=0.5
    bGravityOverrideInSimSpace=true
    bOverrideAngularDamping=true
    bUseGravityOverride=true
    alphaBoolBlend {blendOption=Linear}
    boundBone {boneName=neck_chain_04}
    componentLinearAccScale=<1 1 1>
    componentLinearVelScale=<1 1 1>
    gravityOverride {z=-980}
    physicsBodyDefinitions [{
      sphereCollisionRadius=0
      boundBone {boneName=neck_chain_04}
      boxExtents=<10 20 20>
      constraintSetup {angularLimitsMax=<10 120 10>, angularLimitsMin=<-10 -120 -50>, angularTarget=<0 0 0>}
      localJointOffset=<5 0 0>
     }]
   }
   n6 = Anim Dynamics - Bone: lowerarm_l_chain_01(ComponentPose=@n5, Alpha=1, ExternalForce="0,0,0") {
    angularDampingOverride=0.9999
    angularSpringConstant=20
    bAngularSpring=true
    bGravityOverrideInSimSpace=true
    bOverrideAngularDamping=true
    bUseGravityOverride=true
    alphaBoolBlend {blendOption=Linear}
    boundBone {boneName=lowerarm_l_chain_01}
    componentLinearAccScale=<1 1 1>
    componentLinearVelScale=<1 1 1>
    gravityOverride {z=-980}
    physicsBodyDefinitions [{
      sphereCollisionRadius=0
      boundBone {boneName=lowerarm_l_chain_01}
      boxExtents=<10 20 20>
      constraintSetup {angularLimitsMax=<20 100 100>, angularLimitsMin=<-20 -100 -100>, angularTarget=<0 0 0>}
      localJointOffset=<5 0 0>
     }]
   }
   n7 = Anim Dynamics - Bone: lowerarm_l_chain_02(ComponentPose=@n6, Alpha=1, ExternalForce="0,0,0") {
    angularDampingOverride=0.9999
    angularSpringConstant=20
    bAngularSpring=true
    bGravityOverrideInSimSpace=true
    bOverrideAngularDamping=true
    bUseGravityOverride=true
    alphaBoolBlend {blendOption=Linear}
    boundBone {boneName=lowerarm_l_chain_02}
    componentLinearAccScale=<1 1 1>
    componentLinearVelScale=<1 1 1>
    gravityOverride {z=-980}
    physicsBodyDefinitions [{
      sphereCollisionRadius=0
      boundBone {boneName=lowerarm_l_chain_02}
      boxExtents=<10 20 20>
      constraintSetup {angularLimitsMax=<20 100 100>, angularLimitsMin=<-20 -100 -100>, angularTarget=<0 0 0>}
      localJointOffset=<5 0 0>
     }]
   }
   n8 = Anim Dynamics - Bone: lowerarm_l_chain_03(ComponentPose=@n7, Alpha=1, ExternalForce="0,0,0") {
    angularDampingOverride=0.9999
    angularSpringConstant=20
    bAngularSpring=true
    bGravityOverrideInSimSpace=true
    bOverrideAngularDamping=true
    bUseGravityOverride=true
    alphaBoolBlend {blendOption=Linear}
    boundBone {boneName=lowerarm_l_chain_03}
    componentLinearAccScale=<1 1 1>
    componentLinearVelScale=<1 1 1>
    gravityOverride {z=-980}
    physicsBodyDefinitions [{
      sphereCollisionRadius=0
      boundBone {boneName=lowerarm_l_chain_03}
      boxExtents=<10 20 20>
      constraintSetup {angularLimitsMax=<20 100 100>, angularLimitsMin=<-20 -100 -100>, angularTarget=<0 0 0>}
      localJointOffset=<5 0 0>
     }]
   }
   n9 = Anim Dynamics - Bone: lowerarm_l_chain_04(ComponentPose=@n8, Alpha=1, ExternalForce="0,0,0") {
    angularDampingOverride=0.9999
    angularSpringConstant=20
    bAngularSpring=true
    bGravityOverrideInSimSpace=true
    bOverrideAngularDamping=true
    bUseGravityOverride=true
    alphaBoolBlend {blendOption=Linear}
    boundBone {boneName=lowerarm_l_chain_04}
    componentLinearAccScale=<1 1 1>
    componentLinearVelScale=<1 1 1>
    gravityOverride {z=-980}
    physicsBodyDefinitions [{
      sphereCollisionRadius=0
      boundBone {boneName=lowerarm_l_chain_04}
      boxExtents=<10 20 20>
      constraintSetup {angularLimitsMax=<20 100 100>, angularLimitsMin=<-20 -100 -100>, angularTarget=<0 0 0>}
      localJointOffset=<5 0 0>
     }]
   }
   n10 = Anim Dynamics - Bone: lowerarm_r_chain_01(ComponentPose=@n9, Alpha=1, ExternalForce="0,0,0") {
    angularDampingOverride=0.9999
    angularSpringConstant=10
    bAngularSpring=true
    bGravityOverrideInSimSpace=true
    bOverrideAngularDamping=true
    bUseGravityOverride=true
    alphaBoolBlend {blendOption=Linear}
    boundBone {boneName=lowerarm_r_chain_01}
    componentLinearAccScale=<1 1 1>
    componentLinearVelScale=<1 1 1>
    gravityOverride {z=-980}
    physicsBodyDefinitions [{
      sphereCollisionRadius=0
      boundBone {boneName=lowerarm_r_chain_01}
      boxExtents=<10 20 20>
      constraintSetup {angularLimitsMax=<20 100 100>, angularLimitsMin=<-20 -100 -100>, angularTarget=<0 0 0>}
      localJointOffset=<5 0 0>
     }]
   }
   n11 = Anim Dynamics - Bone: lowerarm_r_chain_02(ComponentPose=@n10, Alpha=1, ExternalForce="0,0,0") {
    angularDampingOverride=0.9999
    bGravityOverrideInSimSpace=true
    bOverrideAngularDamping=true
    bUseGravityOverride=true
    alphaBoolBlend {blendOption=Linear}
    boundBone {boneName=lowerarm_r_chain_02}
    componentLinearAccScale=<1 1 1>
    componentLinearVelScale=<1 1 1>
    gravityOverride {z=-980}
    physicsBodyDefinitions [{
      sphereCollisionRadius=0
      boundBone {boneName=lowerarm_r_chain_02}
      boxExtents=<10 20 20>
      constraintSetup {angularLimitsMax=<20 100 100>, angularLimitsMin=<-20 -100 -100>, angularTarget=<0 0 0>}
      localJointOffset=<5 0 0>
     }]
   }
   n12 = Anim Dynamics - Bone: lowerarm_r_chain_03(ComponentPose=@n11, Alpha=1, ExternalForce="0,0,0") {
    angularDampingOverride=0.9999
    bGravityOverrideInSimSpace=true
    bOverrideAngularDamping=true
    bUseGravityOverride=true
    alphaBoolBlend {blendOption=Linear}
    boundBone {boneName=lowerarm_r_chain_03}
    componentLinearAccScale=<1 1 1>
    componentLinearVelScale=<1 1 1>
    gravityOverride {z=-980}
    physicsBodyDefinitions [{
      sphereCollisionRadius=0
      boundBone {boneName=lowerarm_r_chain_03}
      boxExtents=<10 20 20>
      constraintSetup {angularLimitsMax=<20 100 100>, angularLimitsMin=<-20 -100 -100>, angularTarget=<0 0 0>}
      localJointOffset=<5 0 0>
     }]
   }
   n13 = Anim Dynamics - Bone: lowerarm_r_chain_04(ComponentPose=@n12, Alpha=1, ExternalForce="0,0,0") {
    angularDampingOverride=0.9999
    bGravityOverrideInSimSpace=true
    bOverrideAngularDamping=true
    bUseGravityOverride=true
    alphaBoolBlend {blendOption=Linear}
    boundBone {boneName=lowerarm_r_chain_04}
    componentLinearAccScale=<1 1 1>
    componentLinearVelScale=<1 1 1>
    gravityOverride {z=-980}
    physicsBodyDefinitions [{
      sphereCollisionRadius=0
      boundBone {boneName=lowerarm_r_chain_04}
      boxExtents=<10 20 20>
      constraintSetup {angularLimitsMax=<20 100 100>, angularLimitsMin=<-20 -100 -100>, angularTarget=<0 0 0>}
      localJointOffset=<5 0 0>
     }]
   }
   Output Pose(Result=Component To Local(ComponentPose=@n13)) {layerGroup=None}
   n14 = "Constraint - thigh_half_r (R)"(ComponentPose=Local To Component(LocalPose=Copy Pose From Mesh() {bUseAttachedParent=true}), Alpha=0.3, ConstraintWeights_0=1) {
    alphaBoolBlend {blendOption=Linear}
    boneToModify {boneName=thigh_half_r}
    constraintSetup [{transformType=Rotation, targetBone {boneName=thigh_r}}]
    constraintWeights [1]
   }
   n15 = "Constraint - thigh_half_l (R)"(ComponentPose=@n14, Alpha=0.3, ConstraintWeights_0=1) {
    alphaBoolBlend {blendOption=Linear}
    boneToModify {boneName=thigh_half_l}
    constraintSetup [{transformType=Rotation, targetBone {boneName=thigh_l}}]
    constraintWeights [1]
   }
   n16 = "Constraint - lowerarm_l_chain_01 (PPPPP)"!(
    ComponentPose=@n15
    Alpha=1
    ConstraintWeights_0=0.0258
    ConstraintWeights_1:float=0.7062
    ConstraintWeights_2:float=0.0332
    ConstraintWeights_3:float=0.0142
    ConstraintWeights_4:float=0.2206
   ) -> (Pose) {
    alphaBoolBlend {blendOption=Linear}
    boneToModify {boneName=lowerarm_l_chain_01}
    constraintSetup [
     {targetBone {boneName=hand_l}}
     {targetBone {boneName=lowerarm_twist_01_l}}
     {targetBone {boneName=lowerarm_l}}
     {targetBone {boneName=pinky_metacarpal_l}}
     {targetBone {boneName=lowerarm_twist_02_l}}
    ]
    constraintWeights [1,1,1,1,1]
   }
   n17 = "Constraint - lowerarm_r_chain_01 (PPPPP)"!(
    ComponentPose=@n16
    Alpha=1
    ConstraintWeights_0=0.2206
    ConstraintWeights_1:float=0.0142
    ConstraintWeights_2:float=0.0258
    ConstraintWeights_3:float=0.0332
    ConstraintWeights_4:float=0.7062
   ) -> (Pose) {
    alphaBoolBlend {blendOption=Linear}
    boneToModify {boneName=lowerarm_r_chain_01}
    constraintSetup [
     {targetBone {boneName=lowerarm_twist_02_r}}
     {targetBone {boneName=pinky_metacarpal_r}}
     {targetBone {boneName=hand_r}}
     {targetBone {boneName=lowerarm_r}}
     {targetBone {boneName=lowerarm_twist_01_r}}
    ]
    constraintWeights [1,1,1,1,1]
   }
   n18 = "Constraint - dyn_shoulder_rock_l_01 (PPPP)"!(
    ComponentPose=@n17
    Alpha=1
    ConstraintWeights_0=0.3428
    ConstraintWeights_1:float=0.0222
    ConstraintWeights_2:float=0.1552
    ConstraintWeights_3:float=0.4797
   ) -> (Pose) {
    alphaBoolBlend {blendOption=Linear}
    boneToModify {boneName=dyn_shoulder_rock_l_01}
    constraintSetup [
     {targetBone {boneName=upperarm_twist_01_l}}
     {targetBone {boneName=spine_04}}
     {targetBone {boneName=spine_05}}
     {targetBone {boneName=clavicle_l}}
    ]
    constraintWeights [1,1,1,1]
   }
   n19 = "Constraint - dyn_shoulder_rock_l_02 (PP)"!(ComponentPose=@n18, Alpha=1, ConstraintWeights_0=0.6237, ConstraintWeights_1:float=0.3763) -> (Pose) {
    alphaBoolBlend {blendOption=Linear}
    boneToModify {boneName=dyn_shoulder_rock_l_02}
    constraintSetup [{targetBone {boneName=upperarm_twist_01_l}}, {targetBone {boneName=clavicle_l}}]
    constraintWeights [1,1]
   }
   n20 = "Constraint - dyn_shoulder_rock_l_03 (PPP)"!(ComponentPose=@n19, Alpha=1, ConstraintWeights_0=0.077, ConstraintWeights_1:float=0.5981, ConstraintWeights_2:float=0.3249) -> (Pose) {
    alphaBoolBlend {blendOption=Linear}
    boneToModify {boneName=dyn_shoulder_rock_l_03}
    constraintSetup [{targetBone {boneName=spine_03}}, {targetBone {boneName=upperarm_twist_01_l}}, {targetBone {boneName=clavicle_l}}]
    constraintWeights [1,1,1]
   }
   n21 = "Constraint - dyn_shoulder_rock_l_04 (PPP)"!(ComponentPose=@n20, Alpha=1, ConstraintWeights_0=0.346, ConstraintWeights_1:float=0.1265, ConstraintWeights_2:float=0.5276) -> (Pose) {
    alphaBoolBlend {blendOption=Linear}
    boneToModify {boneName=dyn_shoulder_rock_l_04}
    constraintSetup [{targetBone {boneName=upperarm_twist_02_l}}, {targetBone {boneName=upperarm_l}}, {targetBone {boneName=upperarm_twist_01_l}}]
    constraintWeights [1,1,1]
   }
   n22 = "Constraint - dyn_shoulder_rock_l_05 (PPP)"!(ComponentPose=@n21, Alpha=1, ConstraintWeights_0=0.2745, ConstraintWeights_1:float=0.1329, ConstraintWeights_2:float=0.5927) -> (Pose) {
    alphaBoolBlend {blendOption=Linear}
    boneToModify {boneName=dyn_shoulder_rock_l_05}
    constraintSetup [{targetBone {boneName=upperarm_twist_02_l}}, {targetBone {boneName=upperarm_l}}, {targetBone {boneName=upperarm_twist_01_l}}]
    constraintWeights [1,1,1]
   }
   n23 = "Constraint - dyn_shoulder_rock_l_06 (PP)"!(ComponentPose=@n22, Alpha=1, ConstraintWeights_0=0.385, ConstraintWeights_1:float=0.615) -> (Pose) {
    alphaBoolBlend {blendOption=Linear}
    boneToModify {boneName=dyn_shoulder_rock_l_06}
    constraintSetup [{targetBone {boneName=upperarm_twist_01_l}}, {targetBone {boneName=clavicle_l}}]
    constraintWeights [1,1]
   }
   n24 = "Constraint - dyn_shoulder_rock_l_07 (PP)"!(ComponentPose=@n23, Alpha=1, ConstraintWeights_0=0.6267, ConstraintWeights_1:float=0.3733) -> (Pose) {
    alphaBoolBlend {blendOption=Linear}
    boneToModify {boneName=dyn_shoulder_rock_l_07}
    constraintSetup [{targetBone {boneName=upperarm_twist_01_l}}, {targetBone {boneName=clavicle_l}}]
    constraintWeights [1,1]
   }
   n25 = "Constraint - dyn_shoulder_rock_l_08 (PP)"!(ComponentPose=@n24, Alpha=1, ConstraintWeights_0=0.7563, ConstraintWeights_1:float=0.2437) -> (Pose) {
    alphaBoolBlend {blendOption=Linear}
    boneToModify {boneName=dyn_shoulder_rock_l_08}
    constraintSetup [{targetBone {boneName=upperarm_twist_01_l}}, {targetBone {boneName=clavicle_l}}]
    constraintWeights [1,1]
   }
   n26 = "Constraint - dyn_shoulder_rock_l_09 (PPP)"!(ComponentPose=@n25, Alpha=1, ConstraintWeights_0=0.3618, ConstraintWeights_1:float=0.1035, ConstraintWeights_2:float=0.5347) -> (Pose) {
    alphaBoolBlend {blendOption=Linear}
    boneToModify {boneName=dyn_shoulder_rock_l_09}
    constraintSetup [{targetBone {boneName=upperarm_twist_01_l}}, {targetBone {boneName=spine_05}}, {targetBone {boneName=clavicle_l}}]
    constraintWeights [1,1,1]
   }
   n27 = "Constraint - dyn_shoulder_rock_l_010 (PPPP)"!(
    ComponentPose=@n26
    Alpha=1
    ConstraintWeights_0=0.5755
    ConstraintWeights_1:float=0.0098
    ConstraintWeights_2:float=0.0244
    ConstraintWeights_3:float=0.3904
   ) -> (Pose) {
    alphaBoolBlend {blendOption=Linear}
    boneToModify {boneName=dyn_shoulder_rock_l_010}
    constraintSetup [
     {targetBone {boneName=upperarm_twist_01_l}}
     {targetBone {boneName=spine_04}}
     {targetBone {boneName=spine_05}}
     {targetBone {boneName=clavicle_l}}
    ]
    constraintWeights [1,1,1,1]
   }
   n28 = "Constraint - dyn_shoulder_rock_l_011 (PP)"!(ComponentPose=@n27, Alpha=1, ConstraintWeights_0=0.683, ConstraintWeights_1:float=0.317) -> (Pose) {
    alphaBoolBlend {blendOption=Linear}
    boneToModify {boneName=dyn_shoulder_rock_l_011}
    constraintSetup [{targetBone {boneName=upperarm_twist_01_l}}, {targetBone {boneName=clavicle_l}}]
    constraintWeights [1,1]
   }
   n29 = "Constraint - dyn_shoulder_rock_l_012 (PPP)"!(ComponentPose=@n28, Alpha=1, ConstraintWeights_0=0.2807, ConstraintWeights_1:float=0.0537, ConstraintWeights_2:float=0.6655) -> (Pose) {
    alphaBoolBlend {blendOption=Linear}
    boneToModify {boneName=dyn_shoulder_rock_l_012}
    constraintSetup [{targetBone {boneName=upperarm_twist_01_l}}, {targetBone {boneName=spine_05}}, {targetBone {boneName=clavicle_l}}]
    constraintWeights [1,1,1]
   }
   n30 = "Constraint - dyn_shoulder_rock_l_013 (PP)"!(ComponentPose=@n29, Alpha=1, ConstraintWeights_0=0.405, ConstraintWeights_1:float=0.595) -> (Pose) {
    alphaBoolBlend {blendOption=Linear}
    boneToModify {boneName=dyn_shoulder_rock_l_013}
    constraintSetup [{targetBone {boneName=upperarm_twist_01_l}}, {targetBone {boneName=clavicle_l}}]
    constraintWeights [1,1]
   }
   n31 = "Constraint - dyn_shoulder_rock_l_014 (PP)"!(ComponentPose=@n30, Alpha=1, ConstraintWeights_0=0.5541, ConstraintWeights_1:float=0.4459) -> (Pose) {
    alphaBoolBlend {blendOption=Linear}
    boneToModify {boneName=dyn_shoulder_rock_l_014}
    constraintSetup [{targetBone {boneName=upperarm_twist_01_l}}, {targetBone {boneName=clavicle_l}}]
    constraintWeights [1,1]
   }
   n32 = "Constraint - dyn_shoulder_rock_r_01 (PPP)"!(ComponentPose=@n31, Alpha=1, ConstraintWeights_0=0.4416, ConstraintWeights_1:float=0.5346, ConstraintWeights_2:float=0.0238) -> (Pose) {
    alphaBoolBlend {blendOption=Linear}
    boneToModify {boneName=dyn_shoulder_rock_r_01}
    constraintSetup [{targetBone {boneName=clavicle_r}}, {targetBone {boneName=upperarm_twist_01_r}}, {targetBone {boneName=spine_05}}]
    constraintWeights [1,1,1]
   }
   n33 = "Constraint - dyn_shoulder_rock_r_02 (PP)"!(ComponentPose=@n32, Alpha=1, ConstraintWeights_0=0.4459, ConstraintWeights_1:float=0.5541) -> (Pose) {
    alphaBoolBlend {blendOption=Linear}
    boneToModify {boneName=dyn_shoulder_rock_r_02}
    constraintSetup [{targetBone {boneName=clavicle_r}}, {targetBone {boneName=upperarm_twist_01_r}}]
    constraintWeights [1,1]
   }
   n34 = "Constraint - dyn_shoulder_rock_r_03 (PP)"!(ComponentPose=@n33, Alpha=1, ConstraintWeights_0=0.3126, ConstraintWeights_1:float=0.6874) -> (Pose) {
    alphaBoolBlend {blendOption=Linear}
    boneToModify {boneName=dyn_shoulder_rock_r_03}
    constraintSetup [{targetBone {boneName=clavicle_r}}, {targetBone {boneName=upperarm_twist_01_r}}]
    constraintWeights [1,1]
   }
   n35 = "Constraint - dyn_shoulder_rock_r_04 (PPP)"!(ComponentPose=@n34, Alpha=1, ConstraintWeights_0=0.4119, ConstraintWeights_1:float=0.5686, ConstraintWeights_2:float=0.0195) -> (Pose) {
    alphaBoolBlend {blendOption=Linear}
    boneToModify {boneName=dyn_shoulder_rock_r_04}
    constraintSetup [{targetBone {boneName=clavicle_r}}, {targetBone {boneName=upperarm_twist_01_r}}, {targetBone {boneName=spine_05}}]
    constraintWeights [1,1,1]
   }
   n36 = "Constraint - dyn_shoulder_rock_r_05 (PPP)"!(ComponentPose=@n35, Alpha=1, ConstraintWeights_0=0.565, ConstraintWeights_1:float=0.3538, ConstraintWeights_2:float=0.0812) -> (Pose) {
    alphaBoolBlend {blendOption=Linear}
    boneToModify {boneName=dyn_shoulder_rock_r_05}
    constraintSetup [{targetBone {boneName=clavicle_r}}, {targetBone {boneName=upperarm_twist_01_r}}, {targetBone {boneName=spine_05}}]
    constraintWeights [1,1,1]
   }
   n37 = "Constraint - dyn_shoulder_rock_r_06 (PP)"!(ComponentPose=@n36, Alpha=1, ConstraintWeights_0=0.317, ConstraintWeights_1:float=0.683) -> (Pose) {
    alphaBoolBlend {blendOption=Linear}
    boneToModify {boneName=dyn_shoulder_rock_r_06}
    constraintSetup [{targetBone {boneName=clavicle_r}}, {targetBone {boneName=upperarm_twist_01_r}}]
    constraintWeights [1,1]
   }
   n38 = "Constraint - dyn_shoulder_rock_r_07 (PPP)"!(ComponentPose=@n37, Alpha=1, ConstraintWeights_0=0.6705, ConstraintWeights_1:float=0.3115, ConstraintWeights_2:float=0.018) -> (Pose) {
    alphaBoolBlend {blendOption=Linear}
    boneToModify {boneName=dyn_shoulder_rock_r_07}
    constraintSetup [{targetBone {boneName=clavicle_r}}, {targetBone {boneName=upperarm_twist_01_r}}, {targetBone {boneName=spine_05}}]
    constraintWeights [1,1,1]
   }
   n39 = "Constraint - dyn_shoulder_rock_r_08 (PP)"!(ComponentPose=@n38, Alpha=1, ConstraintWeights_0=0.2815, ConstraintWeights_1:float=0.7185) -> (Pose) {
    alphaBoolBlend {blendOption=Linear}
    boneToModify {boneName=dyn_shoulder_rock_r_08}
    constraintSetup [{targetBone {boneName=clavicle_r}}, {targetBone {boneName=upperarm_twist_01_r}}]
    constraintWeights [1,1]
   }
   n40 = "Constraint - dyn_shoulder_rock_r_09 (PPPP)"!(
    ComponentPose=@n39
    Alpha=1
    ConstraintWeights_0=0.1644
    ConstraintWeights_1:float=0.743
    ConstraintWeights_2:float=0.0358
    ConstraintWeights_3:float=0.0569
   ) -> (Pose) {
    alphaBoolBlend {blendOption=Linear}
    boneToModify {boneName=dyn_shoulder_rock_r_09}
    constraintSetup [
     {targetBone {boneName=clavicle_r}}
     {targetBone {boneName=upperarm_twist_01_r}}
     {targetBone {boneName=upperarm_r}}
     {targetBone {boneName=upperarm_twist_02_r}}
    ]
    constraintWeights [1,1,1,1]
   }
   n41 = "Constraint - dyn_hip_ring_MSTR (PPPP)"!(
    ComponentPose=@n40
    Alpha=1
    ConstraintWeights_0=0.1
    ConstraintWeights_1:float=0.0251
    ConstraintWeights_2:float=0.047
    ConstraintWeights_3:float=0.8278
   ) -> (Pose) {
    alphaBoolBlend {blendOption=Linear}
    boneToModify {boneName=dyn_hip_ring_MSTR}
    constraintSetup [{targetBone {boneName=spine_01}}, {targetBone {boneName=spine_02}}, {targetBone {boneName=thigh_twist_01_r}}, {targetBone {boneName=pelvis}}]
    constraintWeights [1,1,1,1]
   }
   Save cached pose 'prerigid'(Pose=Component To Local(ComponentPose=@n41)) {cacheName=prerigid}
   n42 = Get SprintSpeedRemap() -> (SprintSpeedRemap:double)
   n43 = "Transform (Modify) Bone - Bone: dyn_cloth_fl_MSTR"(
    ComponentPose=Local To Component(LocalPose=Use cached pose 'prerigid'())
    Alpha=@n42
    Translation="0.819,-3.631,12.270"
    Rotation="63.640,10.275,5.475"
    Scale="1,1,1"
   ) {
    rotationMode=BMM_Additive
    rotationSpace=BCS_BoneSpace
    translationMode=BMM_Additive
    translationSpace=BCS_BoneSpace
    alphaBoolBlend {blendOption=Linear}
    boneToModify {boneName=dyn_cloth_fl_MSTR}
    rotation=<rot 63.6399 10.2754 5.4751>
    translation=<0.819 -3.631 12.27>
   }
   n44 = "Transform (Modify) Bone - Bone: dyn_cloth_fr_MSTR"(ComponentPose=@n43, Alpha=@n42, Translation="-4.268,-3.592,-11.355", Rotation="-21.932,-24.388,36.837", Scale="1,1,1") {
    rotationMode=BMM_Additive
    rotationSpace=BCS_BoneSpace
    translationMode=BMM_Additive
    translationSpace=BCS_BoneSpace
    alphaBoolBlend {blendOption=Linear}
    boneToModify {boneName=dyn_cloth_fr_MSTR}
    rotation=<rot -21.9319 -24.3876 36.8367>
    translation=<-4.268 -3.592 -11.355>
   }
   n45 = Break Fort Anim Input Skydiving External Force(FortAnimInput_SkydivingExternalForce=Get SkydivingExternalForceInput()) -> (
    bUseSkydivingVectorForce=~
    SkydivingVectorForce=~
    HeadToPelvisDirection=~
    FloatingMultiplier=~
    FloatingAdditive=~
    DivingMultiplier=~
    DivingAdditive=~
    ParachutingMultiplier=~
    ParachutingAdditive=~
   )
   n46 = RigidBody(ComponentPose=@n44, ExternalForce=@n45.SkydivingVectorForce) {
    bClampLinearTranslationLimitToRefPose=true
    bOverrideWorldGravity=true
    overridePhysicsAsset=&/M_MED_Prisoner_Dyn_Skirt_l
    simulationSpace=BaseBoneSpace
    alphaBoolBlend {blendOption=Linear}
    baseBoneRef {boneName=thigh_half_l}
    componentAppliedLinearAccClamp=<2500 2500 2500>
    componentLinearAccScale=<0.4 0.4 0.4>
    componentLinearVelScale=<0.4 0.4 0.4>
    overrideWorldGravity {z=-980}
   }
   n47 = RigidBody(ComponentPose=@n46, ExternalForce=@n45.SkydivingVectorForce) {
    bClampLinearTranslationLimitToRefPose=true
    bOverrideWorldGravity=true
    overridePhysicsAsset=&/M_MED_Prisoner_Dyn_Skirt_r
    simulationSpace=BaseBoneSpace
    alphaBoolBlend {blendOption=Linear}
    baseBoneRef {boneName=thigh_half_r}
    componentAppliedLinearAccClamp=<2500 2500 2500>
    componentLinearAccScale=<0.4 0.4 0.4>
    componentLinearVelScale=<0.4 0.4 0.4>
    overrideWorldGravity {z=-980}
   }
   n48 = RigidBody(ComponentPose=@n47, ExternalForce=@n45.SkydivingVectorForce) {
    bClampLinearTranslationLimitToRefPose=true
    bOverrideWorldGravity=true
    overridePhysicsAsset=&/M_MED_Prisoner_Dyn_Torso
    simulationSpace=BaseBoneSpace
    alphaBoolBlend {blendOption=Linear}
    baseBoneRef {boneName=spine_01}
    componentAppliedLinearAccClamp=<3000 3000 3000>
    componentLinearAccScale=<0.5 0.5 0.5>
    componentLinearVelScale=<0.5 0.5 0.5>
    overrideWorldGravity {z=-980}
   }
   n49 = Component To Local(ComponentPose=@n48)
   n50 = "Transform (Modify) Bone - Bone: dyn_cloth_fr_01"(ComponentPose=Local To Component(LocalPose=Use cached pose 'prerigid'()), Alpha=1, Translation="0,0,0", Rotation="-15,0,0", Scale="1,1,1") {
    rotationMode=BMM_Additive
    rotationSpace=BCS_BoneSpace
    translationMode=BMM_Additive
    translationSpace=BCS_BoneSpace
    alphaBoolBlend {blendOption=Linear}
    boneToModify {boneName=dyn_cloth_fr_01}
    rotation {pitch=-15}
   }
   n51 = "Transform (Modify) Bone - Bone: dyn_cloth_r_01"(ComponentPose=@n50, Alpha=1, Translation="0,0,0", Rotation="-20,0,0", Scale="1,1,1") {
    rotationMode=BMM_Additive
    rotationSpace=BCS_BoneSpace
    translationMode=BMM_Additive
    translationSpace=BCS_BoneSpace
    alphaBoolBlend {blendOption=Linear}
    boneToModify {boneName=dyn_cloth_r_01}
    rotation {pitch=-20}
   }
   n52 = "Transform (Modify) Bone - Bone: dyn_cloth_bl_01"(ComponentPose=@n51, Alpha=@n1, Translation="0,0,0", Rotation="10,35,0.000", Scale="1,1,1") {
    rotationMode=BMM_Additive
    rotationSpace=BCS_BoneSpace
    alphaBoolBlend {blendOption=Linear}
    alphaScaleBiasClamp {bInterpResult=true, interpSpeedDecreasing=5}
    boneToModify {boneName=dyn_cloth_bl_01}
    rotation {pitch=10, yaw=35}
   }
   n53 = "Transform (Modify) Bone - Bone: dyn_cloth_br_01"(ComponentPose=@n52, Alpha=@n1, Translation="0,0,0", Rotation="15,35,0.000", Scale="1,1,1") {
    rotationMode=BMM_Additive
    rotationSpace=BCS_BoneSpace
    alphaBoolBlend {blendOption=Linear}
    alphaScaleBiasClamp {bInterpResult=true, interpSpeedDecreasing=5}
    boneToModify {boneName=dyn_cloth_br_01}
    rotation {pitch=15, yaw=35}
   }
   n54 = "Transform (Modify) Bone - Bone: dyn_cloth_br_03"(ComponentPose=@n53, Alpha=@n1, Translation="0,0,0", Rotation="0.000,60.000,25", Scale="1,1,1") {
    rotationMode=BMM_Additive
    rotationSpace=BCS_BoneSpace
    alphaBoolBlend {blendOption=Linear}
    alphaScaleBiasClamp {bInterpResult=true, interpSpeedDecreasing=5}
    boneToModify {boneName=dyn_cloth_br_03}
    rotation {roll=25, yaw=60}
   }
   n55 = "Transform (Modify) Bone - Bone: dyn_cloth_br_mid_01"(ComponentPose=@n54, Alpha=@n1, Translation="0,0,0", Rotation="-15,35.000,-0.000", Scale="1,1,1") {
    rotationMode=BMM_Additive
    rotationSpace=BCS_BoneSpace
    alphaBoolBlend {blendOption=Linear}
    alphaScaleBiasClamp {bInterpResult=true, interpSpeedDecreasing=5}
    boneToModify {boneName=dyn_cloth_br_mid_01}
    rotation {pitch=-15, yaw=35}
   }
   n56 = "Transform (Modify) Bone - Bone: dyn_cloth_br_mid_03"(ComponentPose=@n55, Alpha=@n1, Translation="0,0,0", Rotation="0.000,45,-25", Scale="1,1,1") {
    rotationMode=BMM_Additive
    rotationSpace=BCS_BoneSpace
    alphaBoolBlend {blendOption=Linear}
    alphaScaleBiasClamp {bInterpResult=true, interpSpeedDecreasing=5}
    boneToModify {boneName=dyn_cloth_br_mid_03}
    rotation {roll=-25, yaw=45}
   }
   n57 = "Transform (Modify) Bone - Bone: dyn_cloth_r_01"(ComponentPose=@n56, Alpha=@n1, Translation="0,0,0", Rotation="0.000,35,0.000", Scale="1,1,1") {
    rotationMode=BMM_Additive
    rotationSpace=BCS_BoneSpace
    alphaBoolBlend {blendOption=Linear}
    alphaScaleBiasClamp {bInterpResult=true, interpSpeedDecreasing=5}
    boneToModify {boneName=dyn_cloth_r_01}
    rotation {yaw=35}
   }
   n58 = "Transform (Modify) Bone - Bone: dyn_cloth_r_03"(ComponentPose=@n57, Alpha=@n1, Translation="0,0,0", Rotation="0.000,25,0.000", Scale="1,1,1") {
    rotationMode=BMM_Additive
    rotationSpace=BCS_BoneSpace
    alphaBoolBlend {blendOption=Linear}
    alphaScaleBiasClamp {bInterpResult=true, interpSpeedDecreasing=5}
    boneToModify {boneName=dyn_cloth_r_03}
    rotation {yaw=25}
   }
   n59 = "Transform (Modify) Bone - Bone: dyn_cloth_fr_MSTR"(ComponentPose=@n58, Alpha=Get thigh_rot_up_r() -> (thigh_rot_up_r:double), Translation="0,0,3", Rotation="0,0,0", Scale="1,1,1") {
    rotationMode=BMM_Additive
    rotationSpace=BCS_BoneSpace
    translationMode=BMM_Additive
    translationSpace=BCS_BoneSpace
    alphaBoolBlend {blendOption=Linear}
    boneToModify {boneName=dyn_cloth_fr_MSTR}
    translation {z=3}
   }
   n60 = Get thigh_rot_otr_r() -> (thigh_rot_otr_r:double)
   n61 = "Transform (Modify) Bone - Bone: dyn_cloth_fr_01"(ComponentPose=@n59, Alpha=@n60, Translation="0,0,0", Rotation="45,0,0", Scale="1,1,1") {
    rotationMode=BMM_Additive
    rotationSpace=BCS_BoneSpace
    translationMode=BMM_Additive
    translationSpace=BCS_BoneSpace
    alphaBoolBlend {blendOption=Linear}
    boneToModify {boneName=dyn_cloth_fr_01}
    rotation {pitch=45}
   }
   n62 = "Transform (Modify) Bone - Bone: dyn_cloth_fr_01"(ComponentPose=@n61, Alpha=@n60, Translation="0,0,2", Rotation="30,0,0", Scale="1,1,1") {
    rotationMode=BMM_Additive
    rotationSpace=BCS_BoneSpace
    translationMode=BMM_Additive
    translationSpace=BCS_BoneSpace
    alphaBoolBlend {blendOption=Linear}
    boneToModify {boneName=dyn_cloth_fr_01}
    rotation {pitch=30}
    translation {z=2}
   }
   n63 = "Transform (Modify) Bone - Bone: dyn_cloth_r_02"(ComponentPose=@n62, Alpha=@n60, Translation="0,0,0", Rotation="45,0,0", Scale="1,1,1") {
    rotationMode=BMM_Additive
    rotationSpace=BCS_BoneSpace
    translationMode=BMM_Additive
    translationSpace=BCS_BoneSpace
    alphaBoolBlend {blendOption=Linear}
    boneToModify {boneName=dyn_cloth_r_02}
    rotation {pitch=45}
   }
   n64 = Get thigh_rot_up_l() -> (thigh_rot_up_l:double)
   n65 = "Transform (Modify) Bone - Bone: dyn_cloth_fl_01"(ComponentPose=@n63, Alpha=@n64, Translation="0,3,0", Rotation="12,10,0", Scale="1,1,1") {
    rotationMode=BMM_Additive
    rotationSpace=BCS_BoneSpace
    translationMode=BMM_Additive
    translationSpace=BCS_BoneSpace
    alphaBoolBlend {blendOption=Linear}
    boneToModify {boneName=dyn_cloth_fl_01}
    rotation {pitch=12, yaw=10}
    translation {y=3}
   }
   n66 = "Transform (Modify) Bone - Bone: dyn_cloth_l_mid_08"(ComponentPose=@n65, Alpha=@n64, Translation="0,0,0", Rotation="20,-15,0", Scale="1,1,1") {
    rotationMode=BMM_Additive
    rotationSpace=BCS_BoneSpace
    translationMode=BMM_Additive
    translationSpace=BCS_BoneSpace
    alphaBoolBlend {blendOption=Linear}
    boneToModify {boneName=dyn_cloth_l_mid_08}
    rotation {pitch=20, yaw=-15}
   }
   n67 = "Transform (Modify) Bone - Bone: dyn_cloth_l_mid_04"(ComponentPose=@n66, Alpha=@n64, Translation="0,0,0", Rotation="25,-10,0", Scale="1,1,1") {
    rotationMode=BMM_Additive
    rotationSpace=BCS_BoneSpace
    translationMode=BMM_Additive
    translationSpace=BCS_BoneSpace
    alphaBoolBlend {blendOption=Linear}
    boneToModify {boneName=dyn_cloth_l_mid_04}
    rotation {pitch=25, yaw=-10}
   }
   n68 = "Transform (Modify) Bone - Bone: dyn_cloth_fl_02"(ComponentPose=@n67, Alpha=@n64, Translation="0,0,0", Rotation="-5,-25.000,0.000", Scale="1,1,1") {
    rotationMode=BMM_Additive
    rotationSpace=BCS_BoneSpace
    translationMode=BMM_Additive
    translationSpace=BCS_BoneSpace
    alphaBoolBlend {blendOption=Linear}
    boneToModify {boneName=dyn_cloth_fl_02}
    rotation {pitch=-5, yaw=-25}
   }
   n69 = "Transform (Modify) Bone - Bone: dyn_cloth_fl_04"(ComponentPose=@n68, Alpha=@n64, Translation="0,0,0", Rotation="-15,-35,0.000", Scale="1,1,1") {
    rotationMode=BMM_Additive
    rotationSpace=BCS_BoneSpace
    translationMode=BMM_Additive
    translationSpace=BCS_BoneSpace
    alphaBoolBlend {blendOption=Linear}
    boneToModify {boneName=dyn_cloth_fl_04}
    rotation {pitch=-15, yaw=-35}
   }
   n70 = "Transform (Modify) Bone - Bone: dyn_cloth_l_01"(ComponentPose=@n69, Alpha=@n64, Translation="0,-2,3", Rotation="5,15,0", Scale="1,1,1") {
    rotationMode=BMM_Additive
    rotationSpace=BCS_BoneSpace
    translationMode=BMM_Additive
    translationSpace=BCS_BoneSpace
    alphaBoolBlend {blendOption=Linear}
    boneToModify {boneName=dyn_cloth_l_01}
    rotation {pitch=5, yaw=15}
    translation {y=-2, z=3}
   }
   n71 = "Transform (Modify) Bone - Bone: dyn_cloth_l_04"(ComponentPose=@n70, Alpha=@n64, Translation="0,-2,3", Rotation="35,15,0", Scale="1,1,1") {
    rotationMode=BMM_Additive
    rotationSpace=BCS_BoneSpace
    translationMode=BMM_Additive
    translationSpace=BCS_BoneSpace
    alphaBoolBlend {blendOption=Linear}
    boneToModify {boneName=dyn_cloth_l_04}
    rotation {pitch=35, yaw=15}
    translation {y=-2, z=3}
   }
   n72 = "Transform (Modify) Bone - Bone: dyn_cloth_bl_01"(ComponentPose=@n71, Alpha=@n64, Translation="0,0,0", Rotation="-5,25,0.000", Scale="1,1,1") {
    rotationMode=BMM_Additive
    rotationSpace=BCS_BoneSpace
    alphaBoolBlend {blendOption=Linear}
    alphaScaleBiasClamp {bInterpResult=true, interpSpeedDecreasing=5}
    boneToModify {boneName=dyn_cloth_bl_01}
    rotation {pitch=-5, yaw=25}
   }
   n73 = "Transform (Modify) Bone - Bone: dyn_cloth_bl_mid_01"(ComponentPose=@n72, Alpha=@n64, Translation="0,0,0", Rotation="-5,15,-0.000", Scale="1,1,1") {
    rotationMode=BMM_Additive
    rotationSpace=BCS_BoneSpace
    alphaBoolBlend {blendOption=Linear}
    alphaScaleBiasClamp {bInterpResult=true, interpSpeedDecreasing=5}
    boneToModify {boneName=dyn_cloth_bl_mid_01}
    rotation {pitch=-5, yaw=15}
   }
   n74 = Get thigh_rot_otr_l() -> (thigh_rot_otr_l:double)
   n75 = "Transform (Modify) Bone - Bone: dyn_cloth_fl_01"(ComponentPose=@n73, Alpha=@n74, Translation="0,0,0", Rotation="45,0,0", Scale="1,1,1") {
    rotationMode=BMM_Additive
    rotationSpace=BCS_BoneSpace
    translationMode=BMM_Additive
    translationSpace=BCS_BoneSpace
    alphaBoolBlend {blendOption=Linear}
    boneToModify {boneName=dyn_cloth_fl_01}
    rotation {pitch=45}
   }
   n76 = "Transform (Modify) Bone - Bone: dyn_cloth_fl_02"(ComponentPose=@n75, Alpha=@n74, Translation="0,0,2", Rotation="30,0,0", Scale="1,1,1") {
    rotationMode=BMM_Additive
    rotationSpace=BCS_BoneSpace
    translationMode=BMM_Additive
    translationSpace=BCS_BoneSpace
    alphaBoolBlend {blendOption=Linear}
    boneToModify {boneName=dyn_cloth_fl_02}
    rotation {pitch=30}
    translation {z=2}
   }
   n77 = "Transform (Modify) Bone - Bone: dyn_cloth_l_02"(ComponentPose=@n76, Alpha=@n74, Translation="0,0,0", Rotation="45,0,0", Scale="1,1,1") {
    rotationMode=BMM_Additive
    rotationSpace=BCS_BoneSpace
    translationMode=BMM_Additive
    translationSpace=BCS_BoneSpace
    alphaBoolBlend {blendOption=Linear}
    boneToModify {boneName=dyn_cloth_l_02}
    rotation {pitch=45}
   }
   n78 = Component To Local(ComponentPose=@n77)
   n79 = Get thigh_rot_inr_l() -> (thigh_rot_inr_l:double)
   n80 = Look At - Bone: dyn_cloth_fl_MSTR(ComponentPose=Local To Component(LocalPose=Use cached pose 'prerigid'()), Alpha=@n79) {
    alphaBoolBlend {blendOption=Linear}
    alphaScaleBiasClamp {bInterpResult=true}
    boneToModify {boneName=dyn_cloth_fl_MSTR}
    lookAtLocation=<0 12 -5>
    lookAtTarget {boneReference {boneName=calf_l}}
    lookAt_Axis {axis=<1 0>}
    lookUp_Axis {axis {x=0}}
   }
   n81 = Look At - Bone: dyn_cloth_l_mid_08(ComponentPose=@n80, Alpha=@n79) {
    alphaBoolBlend {blendOption=Linear}
    alphaScaleBiasClamp {bInterpResult=true}
    boneToModify {boneName=dyn_cloth_l_mid_08}
    lookAtLocation=<0 5 -15>
    lookAtTarget {boneReference {boneName=calf_l}}
    lookAt_Axis {axis=<1 0>}
    lookUp_Axis {axis {x=0}}
   }
   n82 = Look At - Bone: dyn_cloth_l_01(ComponentPose=@n81, Alpha=@n79) {
    alphaBoolBlend {blendOption=Linear}
    alphaScaleBiasClamp {bInterpResult=true}
    boneToModify {boneName=dyn_cloth_l_01}
    lookAtLocation=<0 -7 -15>
    lookAtTarget {boneReference {boneName=calf_l}}
    lookAt_Axis {axis=<1 0>}
    lookUp_Axis {axis {x=0}}
   }
   n83 = Get thigh_rot_inr_r() -> (thigh_rot_inr_r:double)
   n84 = Look At - Bone: dyn_cloth_fr_MSTR(ComponentPose=@n82, Alpha=@n83) {
    alphaBoolBlend {blendOption=Linear}
    alphaScaleBiasClamp {bInterpResult=true}
    boneToModify {boneName=dyn_cloth_fr_MSTR}
    lookAtLocation=<0 -12 5>
    lookAtTarget {boneReference {boneName=calf_r}}
    lookAt_Axis {axis=<1 0>}
    lookUp_Axis {axis {x=0}}
   }
   n85 = Look At - Bone: dyn_cloth_r_mid_09(ComponentPose=@n84, Alpha=@n83) {
    alphaBoolBlend {blendOption=Linear}
    alphaScaleBiasClamp {bInterpResult=true}
    boneToModify {boneName=dyn_cloth_r_mid_09}
    lookAtLocation=<0 -12 5>
    lookAtTarget {boneReference {boneName=calf_r}}
    lookAt_Axis {axis=<1 0>}
    lookUp_Axis {axis {x=0}}
   }
   n86 = Look At - Bone: dyn_cloth_r_01(ComponentPose=@n85, Alpha=@n83) {
    alphaBoolBlend {blendOption=Linear}
    alphaScaleBiasClamp {bInterpResult=true}
    boneToModify {boneName=dyn_cloth_r_01}
    lookAtLocation {x=0, z=11}
    lookAtTarget {boneReference {boneName=calf_r}}
    lookAt_Axis {axis=<1 0>}
    lookUp_Axis {axis {x=0}}
   }
   n87 = Look At - Bone: dyn_cloth_br_mid_01(ComponentPose=@n86, Alpha=@n83) {
    alphaBoolBlend {blendOption=Linear}
    alphaScaleBiasClamp {bInterpResult=true}
    boneToModify {boneName=dyn_cloth_br_mid_01}
    lookAtLocation=<0 20 11>
    lookAtTarget {boneReference {boneName=calf_r}}
    lookAt_Axis {axis=<1 0>}
    lookUp_Axis {axis {x=0}}
   }
   n88 = "Transform (Modify) Bone - Bone: dyn_cloth_fr_MSTR"(ComponentPose=@n87, Alpha=Get thigh_rot_up_r(), Translation="0,0,3", Rotation="0,0,0", Scale="1,1,1") {
    rotationMode=BMM_Additive
    rotationSpace=BCS_BoneSpace
    translationMode=BMM_Additive
    translationSpace=BCS_BoneSpace
    alphaBoolBlend {blendOption=Linear}
    boneToModify {boneName=dyn_cloth_fr_MSTR}
    translation {z=3}
   }
   n89 = Get thigh_rot_otr_r()
   n90 = "Transform (Modify) Bone - Bone: dyn_cloth_fr_01"(ComponentPose=@n88, Alpha=@n89, Translation="0,0,0", Rotation="45,0,0", Scale="1,1,1") {
    rotationMode=BMM_Additive
    rotationSpace=BCS_BoneSpace
    translationMode=BMM_Additive
    translationSpace=BCS_BoneSpace
    alphaBoolBlend {blendOption=Linear}
    boneToModify {boneName=dyn_cloth_fr_01}
    rotation {pitch=45}
   }
   n91 = "Transform (Modify) Bone - Bone: dyn_cloth_fr_01"(ComponentPose=@n90, Alpha=@n89, Translation="0,0,2", Rotation="30,0,0", Scale="1,1,1") {
    rotationMode=BMM_Additive
    rotationSpace=BCS_BoneSpace
    translationMode=BMM_Additive
    translationSpace=BCS_BoneSpace
    alphaBoolBlend {blendOption=Linear}
    boneToModify {boneName=dyn_cloth_fr_01}
    rotation {pitch=30}
    translation {z=2}
   }
   n92 = "Transform (Modify) Bone - Bone: dyn_cloth_r_02"(ComponentPose=@n91, Alpha=@n89, Translation="0,0,0", Rotation="45,0,0", Scale="1,1,1") {
    rotationMode=BMM_Additive
    rotationSpace=BCS_BoneSpace
    translationMode=BMM_Additive
    translationSpace=BCS_BoneSpace
    alphaBoolBlend {blendOption=Linear}
    boneToModify {boneName=dyn_cloth_r_02}
    rotation {pitch=45}
   }
   n93 = Get thigh_rot_otr_l()
   n94 = "Transform (Modify) Bone - Bone: dyn_cloth_fl_01"(ComponentPose=@n92, Alpha=@n93, Translation="0,0,0", Rotation="45,0,0", Scale="1,1,1") {
    rotationMode=BMM_Additive
    rotationSpace=BCS_BoneSpace
    translationMode=BMM_Additive
    translationSpace=BCS_BoneSpace
    alphaBoolBlend {blendOption=Linear}
    boneToModify {boneName=dyn_cloth_fl_01}
    rotation {pitch=45}
   }
   n95 = "Transform (Modify) Bone - Bone: dyn_cloth_fl_02"(ComponentPose=@n94, Alpha=@n93, Translation="0,0,2", Rotation="30,0,0", Scale="1,1,1") {
    rotationMode=BMM_Additive
    rotationSpace=BCS_BoneSpace
    translationMode=BMM_Additive
    translationSpace=BCS_BoneSpace
    alphaBoolBlend {blendOption=Linear}
    boneToModify {boneName=dyn_cloth_fl_02}
    rotation {pitch=30}
    translation {z=2}
   }
   n96 = "Transform (Modify) Bone - Bone: dyn_cloth_l_02"(ComponentPose=@n95, Alpha=@n93, Translation="0,0,0", Rotation="45,0,0", Scale="1,1,1") {
    rotationMode=BMM_Additive
    rotationSpace=BCS_BoneSpace
    translationMode=BMM_Additive
    translationSpace=BCS_BoneSpace
    alphaBoolBlend {blendOption=Linear}
    boneToModify {boneName=dyn_cloth_l_02}
    rotation {pitch=45}
   }
   n97 = Get thigh_rot_up_l()
   n98 = "Transform (Modify) Bone - Bone: dyn_cloth_fl_01"(ComponentPose=@n96, Alpha=@n97, Translation="0,0,0", Rotation="9.096,-4.987,-10.828", Scale="1,1,1") {
    rotationMode=BMM_Additive
    rotationSpace=BCS_BoneSpace
    translationMode=BMM_Additive
    translationSpace=BCS_BoneSpace
    alphaBoolBlend {blendOption=Linear}
    boneToModify {boneName=dyn_cloth_fl_01}
    rotation=<rot 9.096 -4.987 -10.8284>
   }
   n99 = "Transform (Modify) Bone - Bone: dyn_cloth_l_mid_08"(ComponentPose=@n98, Alpha=@n97, Translation="0,0,0", Rotation="9.096,-4.987,-10.828", Scale="1,1,1") {
    rotationMode=BMM_Additive
    rotationSpace=BCS_BoneSpace
    translationMode=BMM_Additive
    translationSpace=BCS_BoneSpace
    alphaBoolBlend {blendOption=Linear}
    boneToModify {boneName=dyn_cloth_l_mid_08}
    rotation=<rot 9.096 -4.987 -10.8284>
   }
   n100 = "Transform (Modify) Bone - Bone: dyn_cloth_l_01"(ComponentPose=@n99, Alpha=@n97, Translation="0,0,0", Rotation="5,0,0", Scale="1,1,1") {
    rotationMode=BMM_Additive
    rotationSpace=BCS_BoneSpace
    translationMode=BMM_Additive
    translationSpace=BCS_BoneSpace
    alphaBoolBlend {blendOption=Linear}
    boneToModify {boneName=dyn_cloth_l_01}
    rotation {pitch=5}
   }
   n101 = Get thigh_rot_up_r()
   n102 = "Transform (Modify) Bone - Bone: dyn_cloth_fr_01"(ComponentPose=@n100, Alpha=@n101, Translation="0,5,0", Rotation="-10.565,15,10.279", Scale="1,1,1") {
    rotationMode=BMM_Additive
    rotationSpace=BCS_BoneSpace
    translationMode=BMM_Additive
    translationSpace=BCS_BoneSpace
    alphaBoolBlend {blendOption=Linear}
    boneToModify {boneName=dyn_cloth_fr_01}
    rotation=<rot -10.565 15 10.279>
    translation {y=5}
   }
   n103 = "Transform (Modify) Bone - Bone: dyn_cloth_r_mid_09"(ComponentPose=@n102, Alpha=@n101, Translation="0,0,0", Rotation="-10,10,10.829", Scale="1,1,1") {
    rotationMode=BMM_Additive
    rotationSpace=BCS_BoneSpace
    translationMode=BMM_Additive
    translationSpace=BCS_BoneSpace
    alphaBoolBlend {blendOption=Linear}
    boneToModify {boneName=dyn_cloth_r_mid_09}
    rotation=<rot -10 10 10.829>
   }
   n104 = "Transform (Modify) Bone - Bone: dyn_cloth_r_01"(ComponentPose=@n103, Alpha=@n101, Translation="0,0,0", Rotation="-5,0.000,0.000", Scale="1,1,1") {
    rotationMode=BMM_Additive
    rotationSpace=BCS_BoneSpace
    translationMode=BMM_Additive
    translationSpace=BCS_BoneSpace
    alphaBoolBlend {blendOption=Linear}
    boneToModify {boneName=dyn_cloth_r_01}
    rotation {pitch=-5}
   }
   n105 = Get bIsMoving2D()
   n106 = "Transform (Modify) Bone - Bone: dyn_cloth_bl_01"!(ComponentPose=@n104, bAlphaBoolEnabled:bool=@n105, Translation="0,0,0", Rotation="0,20,0.000", Scale="1,1,1") -> (Pose) {
    alphaInputType=Bool
    rotationMode=BMM_Additive
    rotationSpace=BCS_BoneSpace
    alphaBoolBlend {blendInTime=0.2, blendOption=Linear, blendOutTime=0.2}
    alphaScaleBiasClamp {bInterpResult=true, interpSpeedDecreasing=2, interpSpeedIncreasing=5}
    boneToModify {boneName=dyn_cloth_bl_01}
    rotation {yaw=20}
   }
   n107 = "Transform (Modify) Bone - Bone: dyn_cloth_br_01"!(ComponentPose=@n106, bAlphaBoolEnabled:bool=@n105, Translation="0,0,0", Rotation="0,20,0.000", Scale="1,1,1") -> (Pose) {
    alphaInputType=Bool
    rotationMode=BMM_Additive
    rotationSpace=BCS_BoneSpace
    alphaBoolBlend {blendInTime=0.2, blendOption=Linear, blendOutTime=0.2}
    alphaScaleBiasClamp {bInterpResult=true, interpSpeedDecreasing=2, interpSpeedIncreasing=5}
    boneToModify {boneName=dyn_cloth_br_01}
    rotation {yaw=20}
   }
   n108 = "Transform (Modify) Bone - Bone: dyn_cloth_bl_mid_01"!(ComponentPose=@n107, bAlphaBoolEnabled:bool=@n105, Translation="0,0,0", Rotation="3,15,0", Scale="1,1,1") -> (Pose) {
    alphaInputType=Bool
    rotationMode=BMM_Additive
    rotationSpace=BCS_BoneSpace
    alphaBoolBlend {blendInTime=0.2, blendOption=Linear, blendOutTime=0.2}
    alphaScaleBiasClamp {bInterpResult=true, interpSpeedDecreasing=2, interpSpeedIncreasing=5}
    boneToModify {boneName=dyn_cloth_bl_mid_01}
    rotation {pitch=3, yaw=15}
   }
   n109 = "Transform (Modify) Bone - Bone: dyn_cloth_br_mid_01"!(ComponentPose=@n108, bAlphaBoolEnabled:bool=@n105, Translation="0,0,0", Rotation="-3,15,0.000", Scale="1,1,1") -> (Pose) {
    alphaInputType=Bool
    rotationMode=BMM_Additive
    rotationSpace=BCS_BoneSpace
    alphaBoolBlend {blendInTime=0.2, blendOption=Linear, blendOutTime=0.2}
    alphaScaleBiasClamp {bInterpResult=true, interpSpeedDecreasing=2, interpSpeedIncreasing=5}
    boneToModify {boneName=dyn_cloth_br_mid_01}
    rotation {pitch=-3, yaw=15}
   }
   n110 = "Transform (Modify) Bone - Bone: dyn_cloth_l_01"!(ComponentPose=@n109, bAlphaBoolEnabled:bool=@n105, Translation="0,0,0", Rotation="5,10,0", Scale="1,1,1") -> (Pose) {
    alphaInputType=Bool
    rotationMode=BMM_Additive
    rotationSpace=BCS_BoneSpace
    alphaBoolBlend {blendInTime=0.2, blendOption=Linear, blendOutTime=0.2}
    alphaScaleBiasClamp {bInterpResult=true, interpSpeedDecreasing=2, interpSpeedIncreasing=5}
    boneToModify {boneName=dyn_cloth_l_01}
    rotation {pitch=5, yaw=10}
   }
   n111 = "Transform (Modify) Bone - Bone: dyn_cloth_r_01"!(ComponentPose=@n110, bAlphaBoolEnabled:bool=@n105, Translation="0,0,0", Rotation="-5,10,0.000", Scale="1,1,1") -> (Pose) {
    alphaInputType=Bool
    rotationMode=BMM_Additive
    rotationSpace=BCS_BoneSpace
    alphaBoolBlend {blendInTime=0.2, blendOption=Linear, blendOutTime=0.2}
    alphaScaleBiasClamp {bInterpResult=true, interpSpeedDecreasing=2, interpSpeedIncreasing=5}
    boneToModify {boneName=dyn_cloth_r_01}
    rotation {pitch=-5, yaw=10}
   }
   n112 = Get jumpingAlpha() -> (jumpingAlpha:double)
   n113 = "Transform (Modify) Bone - Bone: dyn_cloth_bl_01"!(ComponentPose=@n111, Alpha:float=@n112, Translation="0,0,0", Rotation="-5,15,0.000", Scale="1,1,1") -> (Pose) {
    rotationMode=BMM_Additive
    rotationSpace=BCS_BoneSpace
    alphaBoolBlend {blendOption=Linear}
    alphaScaleBiasClamp {bInterpResult=true, interpSpeedDecreasing=2, interpSpeedIncreasing=5}
    boneToModify {boneName=dyn_cloth_bl_01}
    rotation {pitch=-5, yaw=15}
   }
   n114 = "Transform (Modify) Bone - Bone: dyn_cloth_br_01"!(ComponentPose=@n113, Alpha:float=@n112, Translation="0,0,0", Rotation="5,15,0.000", Scale="1,1,1") -> (Pose) {
    rotationMode=BMM_Additive
    rotationSpace=BCS_BoneSpace
    alphaBoolBlend {blendOption=Linear}
    alphaScaleBiasClamp {bInterpResult=true, interpSpeedDecreasing=2, interpSpeedIncreasing=5}
    boneToModify {boneName=dyn_cloth_br_01}
    rotation {pitch=5, yaw=15}
   }
   n115 = "Transform (Modify) Bone - Bone: dyn_cloth_bl_mid_01"!(ComponentPose=@n114, Alpha:float=@n112, Translation="0,0,0", Rotation="-5,15,0", Scale="1,1,1") -> (Pose) {
    rotationMode=BMM_Additive
    rotationSpace=BCS_BoneSpace
    alphaBoolBlend {blendOption=Linear}
    alphaScaleBiasClamp {bInterpResult=true, interpSpeedDecreasing=2, interpSpeedIncreasing=5}
    boneToModify {boneName=dyn_cloth_bl_mid_01}
    rotation {pitch=-5, yaw=15}
   }
   n116 = "Transform (Modify) Bone - Bone: dyn_cloth_l_01"!(ComponentPose=@n115, Alpha:float=@n112, Translation="0,0,0", Rotation="5,15,0", Scale="1,1,1") -> (Pose) {
    rotationMode=BMM_Additive
    rotationSpace=BCS_BoneSpace
    alphaBoolBlend {blendOption=Linear}
    alphaScaleBiasClamp {bInterpResult=true, interpSpeedDecreasing=2, interpSpeedIncreasing=5}
    boneToModify {boneName=dyn_cloth_l_01}
    rotation {pitch=5, yaw=15}
   }
   n117 = "Transform (Modify) Bone - Bone: dyn_cloth_r_01"!(ComponentPose=@n116, Alpha:float=@n112, Translation="0,0,0", Rotation="-5,15,0.000", Scale="1,1,1") -> (Pose) {
    rotationMode=BMM_Additive
    rotationSpace=BCS_BoneSpace
    alphaBoolBlend {blendOption=Linear}
    alphaScaleBiasClamp {bInterpResult=true, interpSpeedDecreasing=2, interpSpeedIncreasing=5}
    boneToModify {boneName=dyn_cloth_r_01}
    rotation {pitch=-5, yaw=15}
   }
   n118 = "Transform (Modify) Bone - Bone: dyn_cloth_br_mid_01"!(ComponentPose=@n117, Alpha:float=@n112, Translation="0,0,0", Rotation="5,15,0.000", Scale="1,1,1") -> (Pose) {
    rotationMode=BMM_Additive
    rotationSpace=BCS_BoneSpace
    alphaBoolBlend {blendOption=Linear}
    alphaScaleBiasClamp {bInterpResult=true, interpSpeedDecreasing=2, interpSpeedIncreasing=5}
    boneToModify {boneName=dyn_cloth_br_mid_01}
    rotation {pitch=5, yaw=15}
   }
   n119 = Get skydive alpha() -> (skydive alpha:double)
   n120 = "Transform (Modify) Bone - Bone: dyn_cloth_br_01"(ComponentPose=@n118, Alpha=@n119, Translation="0,0,0", Rotation="7,15.000,0.000", Scale="1,1,1") {
    rotationMode=BMM_Additive
    rotationSpace=BCS_BoneSpace
    alphaBoolBlend {blendOption=Linear}
    alphaScaleBiasClamp {bInterpResult=true, interpSpeedDecreasing=2, interpSpeedIncreasing=5}
    boneToModify {boneName=dyn_cloth_br_01}
    rotation {pitch=7, yaw=15}
   }
   n121 = "Transform (Modify) Bone - Bone: dyn_cloth_bl_01"(ComponentPose=@n120, Alpha=@n119, Translation="0,0,0", Rotation="-7,15,0.000", Scale="1,1,1") {
    rotationMode=BMM_Additive
    rotationSpace=BCS_BoneSpace
    alphaBoolBlend {blendOption=Linear}
    alphaScaleBiasClamp {bInterpResult=true, interpSpeedDecreasing=2, interpSpeedIncreasing=5}
    boneToModify {boneName=dyn_cloth_bl_01}
    rotation {pitch=-7, yaw=15}
   }
   n122 = "Transform (Modify) Bone - Bone: dyn_cloth_br_mid_01"(ComponentPose=@n121, Alpha=@n119, Translation="0,0,0", Rotation="-7,15.000,0.000", Scale="1,1,1") {
    rotationMode=BMM_Additive
    rotationSpace=BCS_BoneSpace
    alphaBoolBlend {blendOption=Linear}
    alphaScaleBiasClamp {bInterpResult=true, interpSpeedDecreasing=2, interpSpeedIncreasing=5}
    boneToModify {boneName=dyn_cloth_br_mid_01}
    rotation {pitch=-7, yaw=15}
   }
   n123 = "Transform (Modify) Bone - Bone: dyn_cloth_bl_mid_01"(ComponentPose=@n122, Alpha=@n119, Translation="0,0,0", Rotation="7,15,0.000", Scale="1,1,1") {
    rotationMode=BMM_Additive
    rotationSpace=BCS_BoneSpace
    alphaBoolBlend {blendOption=Linear}
    alphaScaleBiasClamp {bInterpResult=true, interpSpeedDecreasing=2, interpSpeedIncreasing=5}
    boneToModify {boneName=dyn_cloth_bl_mid_01}
    rotation {pitch=7, yaw=15}
   }
   n124 = Get ParachuteAlpha() -> (ParachuteAlpha:double)
   n125 = "Transform (Modify) Bone - Bone: dyn_cloth_br_01"(ComponentPose=@n123, Alpha=@n124, Translation="0,0,0", Rotation="7,15.000,0.000", Scale="1,1,1") {
    rotationMode=BMM_Additive
    rotationSpace=BCS_BoneSpace
    alphaBoolBlend {blendOption=Linear}
    alphaScaleBiasClamp {bInterpResult=true, interpSpeedDecreasing=2, interpSpeedIncreasing=5}
    boneToModify {boneName=dyn_cloth_br_01}
    rotation {pitch=7, yaw=15}
   }
   n126 = "Transform (Modify) Bone - Bone: dyn_cloth_bl_01"(ComponentPose=@n125, Alpha=@n124, Translation="0,0,0", Rotation="-7,15,0.000", Scale="1,1,1") {
    rotationMode=BMM_Additive
    rotationSpace=BCS_BoneSpace
    alphaBoolBlend {blendOption=Linear}
    alphaScaleBiasClamp {bInterpResult=true, interpSpeedDecreasing=2, interpSpeedIncreasing=5}
    boneToModify {boneName=dyn_cloth_bl_01}
    rotation {pitch=-7, yaw=15}
   }
   n127 = "Transform (Modify) Bone - Bone: dyn_cloth_br_mid_01"(ComponentPose=@n126, Alpha=@n124, Translation="0,0,0", Rotation="-7,15.000,0.000", Scale="1,1,1") {
    rotationMode=BMM_Additive
    rotationSpace=BCS_BoneSpace
    alphaBoolBlend {blendOption=Linear}
    alphaScaleBiasClamp {bInterpResult=true, interpSpeedDecreasing=2, interpSpeedIncreasing=5}
    boneToModify {boneName=dyn_cloth_br_mid_01}
    rotation {pitch=-7, yaw=15}
   }
   n128 = "Transform (Modify) Bone - Bone: dyn_cloth_bl_mid_01"(ComponentPose=@n127, Alpha=@n124, Translation="0,0,0", Rotation="7,15,0.000", Scale="1,1,1") {
    rotationMode=BMM_Additive
    rotationSpace=BCS_BoneSpace
    alphaBoolBlend {blendOption=Linear}
    alphaScaleBiasClamp {bInterpResult=true, interpSpeedDecreasing=2, interpSpeedIncreasing=5}
    boneToModify {boneName=dyn_cloth_bl_mid_01}
    rotation {pitch=7, yaw=15}
   }
   n129 = "Transform (Modify) Bone - Bone: dyn_cloth_l_01"(ComponentPose=@n128, Alpha=@n124, Translation="0,0,0", Rotation="5,15,0", Scale="1,1,1") {
    rotationMode=BMM_Additive
    rotationSpace=BCS_BoneSpace
    alphaBoolBlend {blendOption=Linear}
    alphaScaleBiasClamp {bInterpResult=true, interpSpeedDecreasing=2, interpSpeedIncreasing=5}
    boneToModify {boneName=dyn_cloth_l_01}
    rotation {pitch=5, yaw=15}
   }
   n130 = "Transform (Modify) Bone - Bone: dyn_cloth_r_01"(ComponentPose=@n129, Alpha=@n124, Translation="0,0,0", Rotation="-5,15,0.000", Scale="1,1,1") {
    rotationMode=BMM_Additive
    rotationSpace=BCS_BoneSpace
    alphaBoolBlend {blendOption=Linear}
    alphaScaleBiasClamp {bInterpResult=true, interpSpeedDecreasing=2, interpSpeedIncreasing=5}
    boneToModify {boneName=dyn_cloth_r_01}
    rotation {pitch=-5, yaw=15}
   }
   n131 = Component To Local(ComponentPose=@n130)
   n132 = Get bIsCrouching()
   n133 = Get BwdMovement() -> (BwdMovement:double)
   n134 = Trail controller - Bone: dyn_cloth_br_05!(
    ComponentPose=Local To Component(LocalPose=Blend Poses by bool(BlendPose_0=@n78, BlendPose_1=@n131, bActiveValue=@n132) {blendType=Linear})
    Alpha=@n133
    FakeVelocity:Vector="0,0,0"
   ) -> (Pose) {
    bActorSpaceFakeVel=true
    bShowTrailLocation=true
    chainLength=4
    debugLifeTime=0.1
    maxDeltaTime=0.0333
    relaxationSpeedScale=3
    alphaScaleBiasClamp {bInterpResult=true}
    baseJoint {boneName=root}
    planarLimits [
     {drivingBone {boneName=root}, planeTransform {location=<0 0 10>, rotation=<rot 0 0 0>, scale=<1 1 1>}}
     {drivingBone {boneName=spine_03}, planeTransform {location=<0 -20 0>, rotation=<rot 0 15 -89.9999>, scale=<1 1 1>}}
    ]
    rotationLimits [{limitMax=<180 180 180>}, {limitMax=<180 180 180>}, {limitMax=<180 180 180>}, {limitMax=<180 180 180>}]
    rotationOffsets [<0 0 0>, <0 0 0>, <0 0 0>, <0 0 0>]
    trailBone {boneName=dyn_cloth_br_05}
    trailRelaxationSpeed {editorCurveData {keys [{arriveTangent=-72.0739, interpMode=RCIM_Cubic, leaveTangent=-72.0739, tangentMode=RCTM_User, time=0, value=35}, <at 1 5 RCIM_Cubic>]}}
   }
   n135 = Trail controller - Bone: dyn_cloth_bl_05!(ComponentPose=@n134, Alpha=@n133, FakeVelocity:Vector="0,0,0") -> (Pose) {
    bActorSpaceFakeVel=true
    bShowTrailLocation=true
    chainLength=4
    debugLifeTime=0.1
    maxDeltaTime=0.0333
    relaxationSpeedScale=3
    alphaScaleBiasClamp {bInterpResult=true}
    baseJoint {boneName=root}
    planarLimits [
     {drivingBone {boneName=root}, planeTransform {location=<0 0 10>, rotation=<rot 0 0 0>, scale=<1 1 1>}}
     {drivingBone {boneName=spine_03}, planeTransform {location=<0 -20 0>, rotation=<rot 0 15 -89.9999>, scale=<1 1 1>}}
    ]
    rotationLimits [{limitMax=<180 180 180>}, {limitMax=<180 180 180>}, {limitMax=<180 180 180>}, {limitMax=<180 180 180>}]
    rotationOffsets [<0 0 0>, <0 0 0>, <0 0 0>, <0 0 0>]
    trailBone {boneName=dyn_cloth_bl_05}
    trailRelaxationSpeed {editorCurveData {keys [{arriveTangent=-72.0739, interpMode=RCIM_Cubic, leaveTangent=-72.0739, tangentMode=RCTM_User, time=0, value=35}, <at 1 5 RCIM_Cubic>]}}
   }
   n136 = Get LeftMovement() -> (LeftMovement:double)
   n137 = Trail controller - Bone: dyn_cloth_fr_06!(ComponentPose=@n135, Alpha=@n136, FakeVelocity:Vector="0,0,0") -> (Pose) {
    bActorSpaceFakeVel=true
    bShowTrailLocation=true
    chainLength=5
    debugLifeTime=0.1
    maxDeltaTime=0.0333
    relaxationSpeedScale=3
    alphaScaleBiasClamp {bInterpResult=true}
    baseJoint {boneName=root}
    planarLimits [
     {drivingBone {boneName=root}, planeTransform {location=<0 0 10>, rotation=<rot 0 0 0>, scale=<1 1 1>}}
     {drivingBone {boneName=spine_03}, planeTransform {location=<0 -20 0>, rotation=<rot 0 15 -89.9999>, scale=<1 1 1>}}
    ]
    rotationLimits [{limitMax=<180 180 180>}, {limitMax=<180 180 180>}, {limitMax=<180 180 180>}, {limitMax=<180 180 180>}, {limitMax=<180 180 180>}]
    rotationOffsets [<0 0 0>, <0 0 0>, <0 0 0>, <0 0 0>, <0 0 0>]
    trailBone {boneName=dyn_cloth_fr_06}
    trailRelaxationSpeed {editorCurveData {keys [{arriveTangent=-72.0739, interpMode=RCIM_Cubic, leaveTangent=-72.0739, tangentMode=RCTM_User, time=0, value=35}, <at 1 5 RCIM_Cubic>]}}
   }
   n138 = Trail controller - Bone: dyn_cloth_r_mid_07!(ComponentPose=@n137, Alpha=@n136, FakeVelocity:Vector="0,0,0") -> (Pose) {
    bActorSpaceFakeVel=true
    bShowTrailLocation=true
    chainLength=5
    debugLifeTime=0.1
    maxDeltaTime=0.0333
    relaxationSpeedScale=3
    alphaScaleBiasClamp {bInterpResult=true}
    baseJoint {boneName=root}
    planarLimits [
     {drivingBone {boneName=root}, planeTransform {location=<0 0 10>, rotation=<rot 0 0 0>, scale=<1 1 1>}}
     {drivingBone {boneName=spine_03}, planeTransform {location=<0 -20 0>, rotation=<rot 0 15 -89.9999>, scale=<1 1 1>}}
    ]
    rotationLimits [{limitMax=<180 180 180>}, {limitMax=<180 180 180>}, {limitMax=<180 180 180>}, {limitMax=<180 180 180>}, {limitMax=<180 180 180>}]
    rotationOffsets [<0 0 0>, <0 0 0>, <0 0 0>, <0 0 0>, <0 0 0>]
    trailBone {boneName=dyn_cloth_r_mid_07}
    trailRelaxationSpeed {editorCurveData {keys [{arriveTangent=-72.0739, interpMode=RCIM_Cubic, leaveTangent=-72.0739, tangentMode=RCTM_User, time=0, value=35}, <at 1 5 RCIM_Cubic>]}}
   }
   n139 = Trail controller - Bone: dyn_cloth_r_06!(ComponentPose=@n138, Alpha=@n136, FakeVelocity:Vector="0,0,0") -> (Pose) {
    bActorSpaceFakeVel=true
    bShowTrailLocation=true
    chainLength=5
    debugLifeTime=0.1
    maxDeltaTime=0.0333
    relaxationSpeedScale=3
    alphaScaleBiasClamp {bInterpResult=true}
    baseJoint {boneName=root}
    planarLimits [
     {drivingBone {boneName=root}, planeTransform {location=<0 0 10>, rotation=<rot 0 0 0>, scale=<1 1 1>}}
     {drivingBone {boneName=spine_03}, planeTransform {location=<0 -20 0>, rotation=<rot 0 15 -89.9999>, scale=<1 1 1>}}
    ]
    rotationLimits [{limitMax=<180 180 180>}, {limitMax=<180 180 180>}, {limitMax=<180 180 180>}, {limitMax=<180 180 180>}, {limitMax=<180 180 180>}]
    rotationOffsets [<0 0 0>, <0 0 0>, <0 0 0>, <0 0 0>, <0 0 0>]
    trailBone {boneName=dyn_cloth_r_06}
    trailRelaxationSpeed {editorCurveData {keys [{arriveTangent=-72.0739, interpMode=RCIM_Cubic, leaveTangent=-72.0739, tangentMode=RCTM_User, time=0, value=35}, <at 1 5 RCIM_Cubic>]}}
   }
   n140 = Get RightMovement() -> (RightMovement:double)
   n141 = Trail controller - Bone: dyn_cloth_fl_06!(ComponentPose=@n139, Alpha=@n140, FakeVelocity:Vector="0,0,0") -> (Pose) {
    bActorSpaceFakeVel=true
    bShowTrailLocation=true
    chainLength=5
    debugLifeTime=0.1
    maxDeltaTime=0.0333
    relaxationSpeedScale=3
    alphaScaleBiasClamp {bInterpResult=true}
    baseJoint {boneName=root}
    planarLimits [
     {drivingBone {boneName=root}, planeTransform {location=<0 0 10>, rotation=<rot 0 0 0>, scale=<1 1 1>}}
     {drivingBone {boneName=spine_03}, planeTransform {location=<0 -20 0>, rotation=<rot 0 15 -89.9999>, scale=<1 1 1>}}
    ]
    rotationLimits [{limitMax=<180 180 180>}, {limitMax=<180 180 180>}, {limitMax=<180 180 180>}, {limitMax=<180 180 180>}, {limitMax=<180 180 180>}]
    rotationOffsets [<0 0 0>, <0 0 0>, <0 0 0>, <0 0 0>, <0 0 0>]
    trailBone {boneName=dyn_cloth_fl_06}
    trailRelaxationSpeed {editorCurveData {keys [{arriveTangent=-72.0739, interpMode=RCIM_Cubic, leaveTangent=-72.0739, tangentMode=RCTM_User, time=0, value=35}, <at 1 5 RCIM_Cubic>]}}
   }
   n142 = Trail controller - Bone: dyn_cloth_l_mid_07!(ComponentPose=@n141, Alpha=@n140, FakeVelocity:Vector="0,0,0") -> (Pose) {
    bActorSpaceFakeVel=true
    bShowTrailLocation=true
    chainLength=5
    debugLifeTime=0.1
    maxDeltaTime=0.0333
    relaxationSpeedScale=3
    alphaScaleBiasClamp {bInterpResult=true}
    baseJoint {boneName=root}
    planarLimits [
     {drivingBone {boneName=root}, planeTransform {location=<0 0 10>, rotation=<rot 0 0 0>, scale=<1 1 1>}}
     {drivingBone {boneName=spine_03}, planeTransform {location=<0 -20 0>, rotation=<rot 0 15 -89.9999>, scale=<1 1 1>}}
    ]
    rotationLimits [{limitMax=<180 180 180>}, {limitMax=<180 180 180>}, {limitMax=<180 180 180>}, {limitMax=<180 180 180>}, {limitMax=<180 180 180>}]
    rotationOffsets [<0 0 0>, <0 0 0>, <0 0 0>, <0 0 0>, <0 0 0>]
    trailBone {boneName=dyn_cloth_l_mid_07}
    trailRelaxationSpeed {editorCurveData {keys [{arriveTangent=-72.0739, interpMode=RCIM_Cubic, leaveTangent=-72.0739, tangentMode=RCTM_User, time=0, value=35}, <at 1 5 RCIM_Cubic>]}}
   }
   n143 = Trail controller - Bone: dyn_cloth_l_06!(ComponentPose=@n142, Alpha=@n140, FakeVelocity:Vector="0,0,0") -> (Pose) {
    bActorSpaceFakeVel=true
    bShowTrailLocation=true
    chainLength=5
    debugLifeTime=0.1
    maxDeltaTime=0.0333
    relaxationSpeedScale=3
    alphaScaleBiasClamp {bInterpResult=true}
    baseJoint {boneName=root}
    planarLimits [
     {drivingBone {boneName=root}, planeTransform {location=<0 0 10>, rotation=<rot 0 0 0>, scale=<1 1 1>}}
     {drivingBone {boneName=spine_03}, planeTransform {location=<0 -20 0>, rotation=<rot 0 15 -89.9999>, scale=<1 1 1>}}
    ]
    rotationLimits [{limitMax=<180 180 180>}, {limitMax=<180 180 180>}, {limitMax=<180 180 180>}, {limitMax=<180 180 180>}, {limitMax=<180 180 180>}]
    rotationOffsets [<0 0 0>, <0 0 0>, <0 0 0>, <0 0 0>, <0 0 0>]
    trailBone {boneName=dyn_cloth_l_06}
    trailRelaxationSpeed {editorCurveData {keys [{arriveTangent=-72.0739, interpMode=RCIM_Cubic, leaveTangent=-72.0739, tangentMode=RCTM_User, time=0, value=35}, <at 1 5 RCIM_Cubic>]}}
   }
   n144 = Component To Local(ComponentPose=@n143)
   n145 = Get bRigidBodyEnabled()
   n146 = Get bIsCrouchSprinting()
   n147 = "Transform (Modify) Bone - Bone: dyn_cloth_br_mid_02"!(
    ComponentPose=Local To Component(LocalPose=Blend Poses by bool(BlendPose_0=@n49, BlendPose_1=@n144, bActiveValue=@n145) {blendType=Linear})
    bAlphaBoolEnabled:bool=AND Boolean(A=Get bIsCrouching(), B=NOT Boolean(A=@n146))
    Translation="0,0,0"
    Rotation="-45,20,-40"
    Scale="1,1,1"
   ) -> (Pose) {
    alphaInputType=Bool
    rotationMode=BMM_Additive
    rotationSpace=BCS_BoneSpace
    scaleMode=BMM_Additive
    scaleSpace=BCS_BoneSpace
    translationMode=BMM_Additive
    translationSpace=BCS_BoneSpace
    alphaBoolBlend {blendOption=Linear}
    boneToModify {boneName=dyn_cloth_br_mid_02}
    rotation=<rot -45 20 -40>
   }
   n148 = Get bIsInFrontEnd()
   n149 = "Transform (Modify) Bone - Bone: dyn_cloth_fl_05"!(ComponentPose=@n147, bAlphaBoolEnabled:bool=@n148, Translation="0,5,0", Rotation="0,-10,0", Scale="1,1,1") -> (Pose) {
    alphaInputType=Bool
    rotationMode=BMM_Additive
    rotationSpace=BCS_BoneSpace
    scaleMode=BMM_Additive
    scaleSpace=BCS_BoneSpace
    translationMode=BMM_Additive
    translationSpace=BCS_BoneSpace
    alphaBoolBlend {blendOption=Linear}
    boneToModify {boneName=dyn_cloth_fl_05}
    rotation {yaw=-10}
    translation {y=5}
   }
   n150 = "Transform (Modify) Bone - Bone: dyn_cloth_fl_06"!(ComponentPose=@n149, bAlphaBoolEnabled:bool=@n148, Translation="0,0,0", Rotation="29.499,-84.275,-78.491", Scale="1,1,1") -> (Pose) {
    alphaInputType=Bool
    rotationMode=BMM_Additive
    rotationSpace=BCS_BoneSpace
    scaleMode=BMM_Additive
    scaleSpace=BCS_BoneSpace
    translationMode=BMM_Additive
    translationSpace=BCS_BoneSpace
    alphaBoolBlend {blendOption=Linear}
    boneToModify {boneName=dyn_cloth_fl_06}
    rotation=<rot 29.499 -84.275 -78.491>
   }
   n151 = Get SprintSpeedRemap()
   n152 = "Transform (Modify) Bone - Bone: dyn_cloth_fl_MSTR"(
    ComponentPose=Local To Component(LocalPose=Use cached pose 'prerigid'())
    Alpha=@n151
    Translation="0.819,-3.631,12.270"
    Rotation="63.640,10.275,5.475"
    Scale="1,1,1"
   ) {
    rotationMode=BMM_Additive
    rotationSpace=BCS_BoneSpace
    translationMode=BMM_Additive
    translationSpace=BCS_BoneSpace
    alphaBoolBlend {blendOption=Linear}
    boneToModify {boneName=dyn_cloth_fl_MSTR}
    rotation=<rot 63.6399 10.2754 5.4751>
    translation=<0.819 -3.631 12.27>
   }
   n153 = "Transform (Modify) Bone - Bone: dyn_cloth_fr_MSTR"(ComponentPose=@n152, Alpha=@n151, Translation="-4.268,-3.592,-11.355", Rotation="-21.932,-24.388,36.837", Scale="1,1,1") {
    rotationMode=BMM_Additive
    rotationSpace=BCS_BoneSpace
    translationMode=BMM_Additive
    translationSpace=BCS_BoneSpace
    alphaBoolBlend {blendOption=Linear}
    boneToModify {boneName=dyn_cloth_fr_MSTR}
    rotation=<rot -21.9319 -24.3876 36.8367>
    translation=<-4.268 -3.592 -11.355>
   }
   n154 = Break Fort Anim Input Skydiving External Force(FortAnimInput_SkydivingExternalForce=Get SkydivingExternalForceInput()) -> (
    bUseSkydivingVectorForce=~
    SkydivingVectorForce=~
    HeadToPelvisDirection=~
    FloatingMultiplier=~
    FloatingAdditive=~
    DivingMultiplier=~
    DivingAdditive=~
    ParachutingMultiplier=~
    ParachutingAdditive=~
   )
   n155 = RigidBody(ComponentPose=@n153, ExternalForce=@n154.SkydivingVectorForce) {
    bClampLinearTranslationLimitToRefPose=true
    bOverrideWorldGravity=true
    overridePhysicsAsset=&/M_MED_Prisoner_Dyn_Skirt_Skydiving
    simulationSpace=BaseBoneSpace
    alphaBoolBlend {blendOption=Linear}
    baseBoneRef {boneName=pelvis}
    componentAppliedLinearAccClamp=<2500 2500 2500>
    componentLinearAccScale=<0.4 0.4 0.4>
    componentLinearVelScale=<0.4 0.4 0.4>
    overrideWorldGravity {z=-980}
    simSpaceSettings {
     maxAngularAcceleration=1000
     maxAngularVelocity=3
     maxLinearAcceleration=700
     maxLinearVelocity=500
     velocityScaleZ=0
     externalLinearDragV=<0.2 0.2>
    }
   }
   n156 = RigidBody(ComponentPose=@n155, ExternalForce=@n154.SkydivingVectorForce) {
    bClampLinearTranslationLimitToRefPose=true
    bOverrideWorldGravity=true
    overridePhysicsAsset=&M_MED_Prisoner_Dyn_Torso
    simulationSpace=BaseBoneSpace
    alphaBoolBlend {blendOption=Linear}
    baseBoneRef {boneName=spine_01}
    componentAppliedLinearAccClamp=<3000 3000 3000>
    componentLinearAccScale=<0.5 0.5 0.5>
    componentLinearVelScale=<0.5 0.5 0.5>
    overrideWorldGravity {z=-980}
   }
   n157 = "Constraint - dyn_cloth_bl_mid_02 (P)"(ComponentPose=@n156, Alpha=0.25, ConstraintWeights_0=1) {
    alphaBoolBlend {blendOption=Linear}
    boneToModify {boneName=dyn_cloth_bl_mid_02}
    constraintSetup [{targetBone {boneName=thigh_l}}]
    constraintWeights [1]
   }
   n158 = "Constraint - dyn_cloth_bl_mid_03 (P)"(ComponentPose=@n157, Alpha=0.8, ConstraintWeights_0=1) {
    alphaBoolBlend {blendOption=Linear}
    boneToModify {boneName=dyn_cloth_bl_mid_03}
    constraintSetup [{targetBone {boneName=thigh_l}}]
    constraintWeights [1]
   }
   n159 = "Constraint - dyn_cloth_bl_mid_04 (P)"(ComponentPose=@n158, Alpha=0.8, ConstraintWeights_0=1) {
    alphaBoolBlend {blendOption=Linear}
    boneToModify {boneName=dyn_cloth_bl_mid_04}
    constraintSetup [{targetBone {boneName=thigh_l}}]
    constraintWeights [1]
   }
   n160 = "Constraint - dyn_cloth_bl_mid_05 (P)"(ComponentPose=@n159, Alpha=0.8, ConstraintWeights_0=1) {
    alphaBoolBlend {blendOption=Linear}
    boneToModify {boneName=dyn_cloth_bl_mid_05}
    constraintSetup [{targetBone {boneName=thigh_l}}]
    constraintWeights [1]
   }
   n161 = "Constraint - dyn_cloth_bl_mid_07 (P)"(ComponentPose=@n160, Alpha=0.8, ConstraintWeights_0=1) {
    alphaBoolBlend {blendOption=Linear}
    boneToModify {boneName=dyn_cloth_bl_mid_07}
    constraintSetup [{targetBone {boneName=thigh_l}}]
    constraintWeights [1]
   }
   n162 = "Constraint - dyn_cloth_bl_02 (P)"(ComponentPose=@n161, Alpha=0.25, ConstraintWeights_0=1) {
    alphaBoolBlend {blendOption=Linear}
    boneToModify {boneName=dyn_cloth_bl_02}
    constraintSetup [{targetBone {boneName=thigh_l}}]
    constraintWeights [1]
   }
   n163 = "Constraint - dyn_cloth_bl_03 (P)"(ComponentPose=@n162, Alpha=0.4, ConstraintWeights_0=1) {
    alphaBoolBlend {blendOption=Linear}
    boneToModify {boneName=dyn_cloth_bl_03}
    constraintSetup [{targetBone {boneName=thigh_l}}]
    constraintWeights [1]
   }
   n164 = "Constraint - dyn_cloth_bl_04 (P)"(ComponentPose=@n163, Alpha=0.6, ConstraintWeights_0=1) {
    alphaBoolBlend {blendOption=Linear}
    boneToModify {boneName=dyn_cloth_bl_04}
    constraintSetup [{targetBone {boneName=thigh_l}}]
    constraintWeights [1]
   }
   n165 = "Constraint - dyn_cloth_bl_05 (P)"(ComponentPose=@n164, Alpha=0.6, ConstraintWeights_0=1) {
    alphaBoolBlend {blendOption=Linear}
    boneToModify {boneName=dyn_cloth_bl_05}
    constraintSetup [{targetBone {boneName=thigh_l}}]
    constraintWeights [1]
   }
   n166 = "Constraint - dyn_cloth_br_mid_02 (P)"(ComponentPose=@n165, Alpha=0.25, ConstraintWeights_0=1) {
    alphaBoolBlend {blendOption=Linear}
    boneToModify {boneName=dyn_cloth_br_mid_02}
    constraintSetup [{targetBone {boneName=thigh_r}}]
    constraintWeights [1]
   }
   n167 = "Constraint - dyn_cloth_br_mid_03 (P)"(ComponentPose=@n166, Alpha=0.8, ConstraintWeights_0=1) {
    alphaBoolBlend {blendOption=Linear}
    boneToModify {boneName=dyn_cloth_br_mid_03}
    constraintSetup [{targetBone {boneName=thigh_r}}]
    constraintWeights [1]
   }
   n168 = "Constraint - dyn_cloth_br_mid_04 (P)"(ComponentPose=@n167, Alpha=0.8, ConstraintWeights_0=1) {
    alphaBoolBlend {blendOption=Linear}
    boneToModify {boneName=dyn_cloth_br_mid_04}
    constraintSetup [{targetBone {boneName=thigh_r}}]
    constraintWeights [1]
   }
   n169 = "Constraint - dyn_cloth_br_mid_05 (P)"(ComponentPose=@n168, Alpha=0.8, ConstraintWeights_0=1) {
    alphaBoolBlend {blendOption=Linear}
    boneToModify {boneName=dyn_cloth_br_mid_05}
    constraintSetup [{targetBone {boneName=thigh_r}}]
    constraintWeights [1]
   }
   n170 = "Constraint - dyn_cloth_br_mid_07 (P)"(ComponentPose=@n169, Alpha=0.8, ConstraintWeights_0=1) {
    alphaBoolBlend {blendOption=Linear}
    boneToModify {boneName=dyn_cloth_br_mid_07}
    constraintSetup [{targetBone {boneName=thigh_r}}]
    constraintWeights [1]
   }
   n171 = "Constraint - dyn_cloth_br_02 (P)"(ComponentPose=@n170, Alpha=0.25, ConstraintWeights_0=1) {
    alphaBoolBlend {blendOption=Linear}
    boneToModify {boneName=dyn_cloth_br_02}
    constraintSetup [{targetBone {boneName=thigh_r}}]
    constraintWeights [1]
   }
   n172 = "Constraint - dyn_cloth_br_03 (P)"(ComponentPose=@n171, Alpha=0.4, ConstraintWeights_0=1) {
    alphaBoolBlend {blendOption=Linear}
    boneToModify {boneName=dyn_cloth_br_03}
    constraintSetup [{targetBone {boneName=thigh_r}}]
    constraintWeights [1]
   }
   n173 = "Constraint - dyn_cloth_br_04 (P)"(ComponentPose=@n172, Alpha=0.6, ConstraintWeights_0=1) {
    alphaBoolBlend {blendOption=Linear}
    boneToModify {boneName=dyn_cloth_br_04}
    constraintSetup [{targetBone {boneName=thigh_r}}]
    constraintWeights [1]
   }
   n174 = "Constraint - dyn_cloth_br_05 (P)"(ComponentPose=@n173, Alpha=0.6, ConstraintWeights_0=1) {
    alphaBoolBlend {blendOption=Linear}
    boneToModify {boneName=dyn_cloth_br_05}
    constraintSetup [{targetBone {boneName=thigh_r}}]
    constraintWeights [1]
   }
   n175 = "Transform (Modify) Bone - Bone: dyn_cloth_bl_mid_04"(ComponentPose=@n174, Alpha=1, Translation="0,6,-9", Rotation="0,0,0", Scale="1,1,1") {
    rotationSpace=BCS_BoneSpace
    translationMode=BMM_Additive
    translationSpace=BCS_BoneSpace
    alphaBoolBlend {blendOption=Linear}
    boneToModify {boneName=dyn_cloth_bl_mid_04}
    translation {y=6, z=-9}
   }
   n176 = "Transform (Modify) Bone - Bone: dyn_cloth_bl_mid_05"(ComponentPose=@n175, Alpha=1, Translation="0,9,-9", Rotation="0,0,0", Scale="1,1,1") {
    rotationSpace=BCS_BoneSpace
    translationMode=BMM_Additive
    translationSpace=BCS_BoneSpace
    alphaBoolBlend {blendOption=Linear}
    boneToModify {boneName=dyn_cloth_bl_mid_05}
    translation {y=9, z=-9}
   }
   n177 = "Transform (Modify) Bone - Bone: dyn_cloth_bl_mid_07"(ComponentPose=@n176, Alpha=1, Translation="0,9,-9", Rotation="0,0,0", Scale="1,1,1") {
    rotationSpace=BCS_BoneSpace
    translationMode=BMM_Additive
    translationSpace=BCS_BoneSpace
    alphaBoolBlend {blendOption=Linear}
    boneToModify {boneName=dyn_cloth_bl_mid_07}
    translation {y=9, z=-9}
   }
   n178 = "Transform (Modify) Bone - Bone: dyn_cloth_bl_03"(ComponentPose=@n177, Alpha=1, Translation="0,6,0", Rotation="0,0,0", Scale="1,1,1") {
    rotationSpace=BCS_BoneSpace
    translationMode=BMM_Additive
    translationSpace=BCS_BoneSpace
    alphaBoolBlend {blendOption=Linear}
    boneToModify {boneName=dyn_cloth_bl_03}
    translation {y=6}
   }
   n179 = "Transform (Modify) Bone - Bone: dyn_cloth_bl_04"(ComponentPose=@n178, Alpha=1, Translation="0,10,0", Rotation="0,0,0", Scale="1,1,1") {
    rotationSpace=BCS_BoneSpace
    translationMode=BMM_Additive
    translationSpace=BCS_BoneSpace
    alphaBoolBlend {blendOption=Linear}
    boneToModify {boneName=dyn_cloth_bl_04}
    translation {y=10}
   }
   n180 = "Transform (Modify) Bone - Bone: dyn_cloth_bl_05"(ComponentPose=@n179, Alpha=1, Translation="0,10,0", Rotation="0,0,0", Scale="1,1,1") {
    rotationSpace=BCS_BoneSpace
    translationMode=BMM_Additive
    translationSpace=BCS_BoneSpace
    alphaBoolBlend {blendOption=Linear}
    boneToModify {boneName=dyn_cloth_bl_05}
    translation {y=10}
   }
   n181 = "Transform (Modify) Bone - Bone: dyn_cloth_br_mid_04"(ComponentPose=@n180, Alpha=1, Translation="0,8,9", Rotation="0,0,0", Scale="1,1,1") {
    translationMode=BMM_Additive
    translationSpace=BCS_BoneSpace
    alphaBoolBlend {blendOption=Linear}
    boneToModify {boneName=dyn_cloth_br_mid_04}
    translation {y=8, z=9}
   }
   n182 = "Transform (Modify) Bone - Bone: dyn_cloth_br_mid_05"(ComponentPose=@n181, Alpha=1, Translation="0,10,9", Rotation="0,0,0", Scale="1,1,1") {
    translationMode=BMM_Additive
    translationSpace=BCS_BoneSpace
    alphaBoolBlend {blendOption=Linear}
    boneToModify {boneName=dyn_cloth_br_mid_05}
    translation {y=10, z=9}
   }
   n183 = "Transform (Modify) Bone - Bone: dyn_cloth_br_mid_07"(ComponentPose=@n182, Alpha=1, Translation="0,10,9", Rotation="0,0,0", Scale="1,1,1") {
    translationMode=BMM_Additive
    translationSpace=BCS_BoneSpace
    alphaBoolBlend {blendOption=Linear}
    boneToModify {boneName=dyn_cloth_br_mid_07}
    translation {y=10, z=9}
   }
   n184 = "Transform (Modify) Bone - Bone: dyn_cloth_br_03"(ComponentPose=@n183, Alpha=1, Translation="0,8,0", Rotation="0,0,0", Scale="1,1,1") {
    translationMode=BMM_Additive
    translationSpace=BCS_BoneSpace
    alphaBoolBlend {blendOption=Linear}
    boneToModify {boneName=dyn_cloth_br_03}
    translation {y=8}
   }
   n185 = "Transform (Modify) Bone - Bone: dyn_cloth_br_04"(ComponentPose=@n184, Alpha=1, Translation="0,13,0", Rotation="0,0,0", Scale="1,1,1") {
    translationMode=BMM_Additive
    translationSpace=BCS_BoneSpace
    alphaBoolBlend {blendOption=Linear}
    boneToModify {boneName=dyn_cloth_br_04}
    translation {y=13}
   }
   n186 = "Transform (Modify) Bone - Bone: dyn_cloth_br_05"(ComponentPose=@n185, Alpha=1, Translation="0,10,0", Rotation="0,0,0", Scale="1,1,1") {
    translationMode=BMM_Additive
    translationSpace=BCS_BoneSpace
    alphaBoolBlend {blendOption=Linear}
    boneToModify {boneName=dyn_cloth_br_05}
    translation {y=10}
   }
   Save cached pose 'post rigid'(Pose=Blend Poses by bool(BlendPose_0=Component To Local(ComponentPose=@n186), BlendPose_1=Component To Local(ComponentPose=@n150), bActiveValue=Get bIsSkyDiving()) {blendType=Linear}) {cacheName=post rigid}
  ]}
 EventGraph {nodes [
   Event BlueprintUpdateAnimation()
   n1 = TryGetPawnOwner()
   > Is Valid(InputObject=@n1)
   > n2 = Sequence!(execute) -> (then_0, then_1, then_2:Exec, then_3:Exec)
   n3 = Get Actor Rotation!(self:Actor Object Reference=@n1) -> (ReturnValue:Rotator) {is=title K2Node_CallFunction}
   n4 = GetVelocity!(self:Actor Object Reference=@n1) -> (ReturnValue:Vector)
   n5 = BreakVector(InVec=vector * vector(A:Vector=Get Rotation X Vector!(InRot:Rotator=@n3) -> (ReturnValue:Vector) {is=title K2Node_CallFunction}, B:Vector=@n4) -> (ReturnValue:Vector) {is=title K2Node_CommutativeAssociativeBinaryOperator})
   n6 = float + float(A=@n5, B=@n5.Y) {is=title K2Node_CommutativeAssociativeBinaryOperator}
   n7 = "Not Equal (Float)"!(A:double=BreakVector(InVec=@n4).Z, B:double=0) -> (ReturnValue:bool) {is=title K2Node_CallFunction}
   > Set SprintSpeedRemap(SprintSpeedRemap:double=float * float(A=MapRangeClamped(Value=@n6, InRangeA=300, InRangeB=550), B=SelectFloat(A=0, B=1, bPickA=@n7))) -> (Output_Get:double)
   n8 = GetOwnerSkeletalMeshComponent()
   n9 = BreakRotator(InRot=BreakTransform(InTransform=GetDeltaTransformFromRefPose(self=@n8, BoneName=thigh_r, BaseName=pelvis)).Rotation)
   Set thigh_rot_up_r(execute=@n2.then_1, thigh_rot_up_r:double=MapRangeClamped(Value=@n9.Yaw, InRangeA=-15, InRangeB=-100)) -> (Output_Get:double)
   > Set thigh_rot_inr_r(thigh_rot_inr_r:double=MapRangeClamped(Value=@n9.Pitch, InRangeA=0, InRangeB=-10, OutRangeA=1, OutRangeB=0)) -> (Output_Get:double)
   > Set thigh_rot_otr_r(thigh_rot_otr_r:double=MapRangeClamped(Value=@n9.Pitch, InRangeA=20, InRangeB=50)) -> (Output_Get:double)
   n10 = GetOwnerSkeletalMeshComponent()
   n11 = BreakRotator(InRot=BreakTransform(InTransform=GetDeltaTransformFromRefPose(self=@n10, BoneName=thigh_l, BaseName=pelvis)).Rotation)
   > Set thigh_rot_up_l(thigh_rot_up_l:double=MapRangeClamped(Value=@n11.Yaw, InRangeA=-10, InRangeB=-100)) -> (Output_Get:double)
   > Set thigh_rot_inr_l(thigh_rot_inr_l:double=MapRangeClamped(Value=@n11.Pitch, InRangeA=0, InRangeB=-10, OutRangeA=1, OutRangeB=0)) -> (Output_Get:double)
   > Set thigh_rot_otr_l(thigh_rot_otr_l:double=MapRangeClamped(Value=@n11.Pitch, InRangeA=20, InRangeB=50)) -> (Output_Get:double)
   > Set jumpingAlpha(jumpingAlpha:double="To Float (Boolean)"(InBool=Get bIsJumping())) -> (Output_Get:double)
   n12 = Get bIsMoving2D()
   > Set isNotCrouchRunning(isNotCrouchRunning:double="To Float (Boolean)"(InBool=NOT Boolean(A=@n12))) -> (Output_Get:double)
   > Set skydive alpha(skydive alpha:double="To Float (Boolean)"(InBool=Get bIsSkyDiving())) -> (Output_Get:double)
   > Set ParachuteAlpha(ParachuteAlpha:double="To Float (Boolean)"(InBool=Get bIsParachuteOpened())) -> (Output_Get:double)
   n13 = PlayerMovement(execute=@n2.then_2, AnimInstance=Self-Reference(), XDotProductCoefficient=1, YDotProductCoefficient=1, ZDotProductCoefficient=1)
   n14 = MapRangeClamped(Value=@n13.YMovement, InRangeA=0, InRangeB=-1)
   n15 = float * float(A="Absolute (Float)"(A=@n13), B=0.5)
   > Set BwdMovement(BwdMovement:double=MapRangeClamped(Value=float + float(A=@n15, B=@n14), InRangeA=0, InRangeB=1.1, OutRangeA=0.4, OutRangeB=0)) -> (Output_Get:double)
   n16 = MapRangeClamped(Value=@n13, InRangeA=-0.2, InRangeB=1)
   n17 = float * float(A="Absolute (Float)"(A=@n13.YMovement), B=0.2)
   > Set RightMovement(RightMovement:double=MapRangeClamped(Value=float + float(A=@n17, B=@n16), InRangeA=0, InRangeB=1.2, OutRangeB=0.5)) -> (Output_Get:double)
   n18 = MapRangeClamped(Value=@n13, InRangeA=0.2, InRangeB=-1, OutRangeB=0.3)
   n19 = float * float(A="Absolute (Float)"(A=@n13.YMovement), B=0.2)
   > Set LeftMovement(LeftMovement:double=MapRangeClamped(Value=float + float(A=@n19, B=@n18), InRangeA=0, InRangeB=1.2, OutRangeB=0.5)) -> (Output_Get:double)
   n20 = Branch(execute=@n2.then_3, Condition=Get bIsSkyDiving())
   > n21 = Branch(Condition="float > float"(A=Get PawnVelocityZ(), B=50) {is=title K2Node_CallFunction})
   > Set BwdMovement(BwdMovement=float * float(A=Get BwdMovement(), B=0))
   Set BwdMovement(execute=[@n20.else, @n21.else], BwdMovement=float * float(A=Get BwdMovement(), B=1))
   Event OnParachuteOpened()
   > n22 = Delay(Duration=1.5)
   ResetDynamics(execute=[@n22.then, @n23.then, @n24.then, @n28.then, @n27.then], InTeleportType=TeleportPhysics)
   > ResetDynamics(InTeleportType=ResetPhysics)
   n23 = Event OnEndSkydiving()
   n24 = Event OnBeginSkydiving()
   n25 = Event OnBeginTargeting() {is=title K2Node_Event}
   n26 = Get bIsCrouching()
   n27 = Branch(execute=[@n25.then, @n29.then], Condition=AND Boolean(A=Get bIsTargeting(), B=NOT Boolean(A=@n26)))
   n28 = Event OnBeginCrouching()
   n29 = Event OnEndCrouching()
   TryGetPawnOwner()
   Get bIsCrouchSprinting()
  ]}
]
variables [
 BwdMovement {type=number}
 isNotCrouchRunning {type=number}
 jumpingAlpha {type=number}
 LeftMovement {type=number}
 ParachuteAlpha {type=number}
 RightMovement {type=number}
 skydive alpha {type=number}
 SprintSpeedRemap {type=number}
 thigh_rot_inr_l {type=number}
 thigh_rot_inr_r {type=number}
 thigh_rot_otr_l {type=number}
 thigh_rot_otr_r {type=number}
 thigh_rot_up_l {type=number}
 thigh_rot_up_r {type=number}
]
